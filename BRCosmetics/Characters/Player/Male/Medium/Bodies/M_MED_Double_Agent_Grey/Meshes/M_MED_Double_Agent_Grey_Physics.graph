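# Physics asset for the male medium Double Agent Grey body, covering three strap bone chains (left, right, back) plus the pelvis, thighs and calves.
# Every body is a single capsule, and all three strap chains hang off the pelvis through chained constraints with tight twist limits between segments.
# Back strap swing limits widen down the chain, and its root is free, while the right strap root is locked.
# Thighs lock to the pelvis and calves rotate freely.
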PhysicsAsset /BRCosmetics/Characters/Player/Male/Medium/Bodies/M_MED_Double_Agent_Grey/Meshes/M_MED_Double_Agent_Grey_Physics
bodies {
 calf_l [{shapeType=Capsule, length=30, radius=6.5, center=<-19.3581 0 0>, rotation=<rot 90 0 0>}]
 calf_r [{shapeType=Capsule, length=30, radius=7, center=<19 0 0>, rotation=<rot 90 0 0>}]
 dyn_L_strap_jnt_1 [{shapeType=Capsule, length=11.5, radius=2.75, center=<4.0515 0 1.9634>, rotation=<rot 84.7797 0 0>}]
 dyn_L_strap_jnt_2 [{shapeType=Capsule, length=10.25, radius=2.25, center=<4 0 1.155>, rotation=<rot 87.2447 -0.0001 -0.0001>}]
 dyn_L_strap_jnt_3 [{shapeType=Capsule, length=16.75, radius=2.25, center=<4.9698 0 0.6218>, rotation=<rot 84.6146 0 0>}]
 dyn_R_strap_jnt_1 [{shapeType=Capsule, length=9.5, radius=2, center=<3 0 0>, rotation=<rot 90 0 0>}]
 dyn_R_strap_jnt_2 [{shapeType=Capsule, length=10.75, radius=2, center=<4 -0.0001 -0.2237>, rotation=<rot 87.8889 180 180>}]
 dyn_R_strap_jnt_3 [{shapeType=Capsule, length=15, radius=2, center=<7.8725 0 -0.1613>, rotation=<rot 89.2551 -180 -180>}]
 dyn_back_strap_jnt_2 [{shapeType=Capsule, length=1, radius=1}]
 dyn_back_strap_jnt_3 [{shapeType=Capsule, length=5.25, radius=4.75, center=<4 0 0>, rotation=<rot 90 0 0>}]
 dyn_back_strap_jnt_4 [{shapeType=Capsule, length=4.75, radius=4.5, center=<5 0 0>, rotation=<rot 90 0 0>}]
 dyn_back_strap_jnt_5 [{shapeType=Capsule, length=5.25, radius=4, center=<6 0 0>, rotation=<rot 90 0 0>}]
 dyn_back_strap_jnt_6 [{shapeType=Capsule, length=4.75, radius=3.75, center=<7 0 0>, rotation=<rot 90 0 0>}]
 dyn_back_strap_jnt_7 [{shapeType=Capsule, length=4.75, radius=3.75, center=<8 0 0>, rotation=<rot 90 0 0>}]
 pelvis [{shapeType=Capsule, length=10, radius=10, center=<-4.0337 0 0>}]
 thigh_l [{shapeType=Capsule, length=30.75, radius=8, center=<-19.6241 0.0134 0.36>, rotation=<rot 89.1162 -1.7502 -3.6637>}]
 thigh_r [{shapeType=Capsule, length=40, radius=8, center=<17.4876 -0.0833 -0.9936>, rotation=<rot 89.1345 -174.8685 -176.7816>}]
}
constraints [
 {
  bone1Name=thigh_l
  bone2Name=pelvis
  swing1LimitDegrees=45
  swing1Motion=Locked
  swing2LimitDegrees=45
  swing2Motion=Locked
  twistLimitDegrees=45
  twistMotion=Locked
 }
 {
  bone1Name=thigh_r
  bone2Name=pelvis
  swing1LimitDegrees=45
  swing1Motion=Locked
  swing2LimitDegrees=45
  swing2Motion=Locked
  twistLimitDegrees=45
  twistMotion=Locked
 }
 {bone1Name=dyn_back_strap_jnt_3, bone2Name=dyn_back_strap_jnt_2, swing1LimitDegrees=18.4286, swing2LimitDegrees=37.2857, twistLimitDegrees=5}
 {
  bone1Name=calf_l
  bone2Name=thigh_l
  swing1LimitDegrees=45
  swing1Motion=Free
  swing2LimitDegrees=45
  swing2Motion=Free
  twistLimitDegrees=45
  twistMotion=Free
 }
 {
  bone1Name=calf_r
  bone2Name=thigh_r
  swing1LimitDegrees=45
  swing1Motion=Free
  swing2LimitDegrees=45
  swing2Motion=Free
  twistLimitDegrees=45
  twistMotion=Free
 }
 {bone1Name=dyn_R_strap_jnt_2, bone2Name=dyn_R_strap_jnt_1, swing1LimitDegrees=45, swing2LimitDegrees=45, twistLimitDegrees=5}
 {bone1Name=dyn_R_strap_jnt_3, bone2Name=dyn_R_strap_jnt_2, swing1LimitDegrees=45, swing2LimitDegrees=45, twistLimitDegrees=5}
 {bone1Name=dyn_L_strap_jnt_2, bone2Name=dyn_L_strap_jnt_1, swing1LimitDegrees=45, swing2LimitDegrees=45, twistLimitDegrees=5}
 {bone1Name=dyn_L_strap_jnt_3, bone2Name=dyn_L_strap_jnt_2, swing1LimitDegrees=45, swing2LimitDegrees=45, twistLimitDegrees=5}
 {
  bone1Name=dyn_R_strap_jnt_1
  bone2Name=pelvis
  swing1LimitDegrees=45
  swing1Motion=Locked
  swing2LimitDegrees=45
  swing2Motion=Locked
  twistLimitDegrees=45
  twistMotion=Locked
 }
 {bone1Name=dyn_L_strap_jnt_1, bone2Name=pelvis, swing1LimitDegrees=45, swing2LimitDegrees=45, twistLimitDegrees=5}
 {bone1Name=dyn_back_strap_jnt_4, bone2Name=dyn_back_strap_jnt_3, swing1LimitDegrees=45, swing2LimitDegrees=57, twistLimitDegrees=5}
 {bone1Name=dyn_back_strap_jnt_5, bone2Name=dyn_back_strap_jnt_4, swing1LimitDegrees=85, swing2LimitDegrees=85, twistLimitDegrees=5}
 {bone1Name=dyn_back_strap_jnt_6, bone2Name=dyn_back_strap_jnt_5, swing1LimitDegrees=85, swing2LimitDegrees=85, twistLimitDegrees=5}
 {bone1Name=dyn_back_strap_jnt_7, bone2Name=dyn_back_strap_jnt_6, swing1LimitDegrees=85, swing2LimitDegrees=85, twistLimitDegrees=5}
 {
  bone1Name=dyn_back_strap_jnt_2
  bone2Name=pelvis
  swing1LimitDegrees=45
  swing1Motion=Free
  swing2LimitDegrees=45
  swing2Motion=Free
  twistLimitDegrees=45
  twistMotion=Free
 }
]
