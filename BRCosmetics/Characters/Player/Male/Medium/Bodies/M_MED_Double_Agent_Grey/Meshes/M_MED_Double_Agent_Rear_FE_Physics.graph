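# Physics asset for the rear loincloth of the male medium Double Agent Grey cosmetic body mesh.
# Left, center and right chains of three capsule bodies hang from small master bodies tied to the pelvis by free constraints, alongside pelvis and thigh capsules.
# Chain links have tight twist limits and wider first-swing limits, while cross constraints tie the center chain's lower two links to both sides.
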PhysicsAsset /BRCosmetics/Characters/Player/Male/Medium/Bodies/M_MED_Double_Agent_Grey/Meshes/M_MED_Double_Agent_Rear_FE_Physics
bodies {
 dyn_loinClothBack_01_c [{shapeType=Capsule, length=8.25, radius=8, center=<4 -6.5 0>}]
 dyn_loinClothBack_01_l [{shapeType=Capsule, length=10, radius=7.5, center=<4 -8.35 0>}]
 dyn_loinClothBack_01_r [{shapeType=Capsule, length=10.5, radius=7.75, center=<4 -8 0.0361>}]
 dyn_loinClothBack_02_c [{shapeType=Capsule, length=8.75, radius=7.75, center=<4 -6.5 0>}]
 dyn_loinClothBack_02_l [{shapeType=Capsule, length=8.5, radius=8.75, center=<4 -8.15 0>}]
 dyn_loinClothBack_02_r [{shapeType=Capsule, length=7.25, radius=7.75, center=<4 -7.5 -0.1>}]
 dyn_loinClothBack_03_c [{shapeType=Capsule, length=8, radius=7, center=<4 -6.5 0>}]
 dyn_loinClothBack_03_l [{shapeType=Capsule, length=8, radius=7, center=<4 -7.75 0>}]
 dyn_loinClothBack_03_r [{shapeType=Capsule, length=8, radius=7, center=<4 -8 0>}]
 dyn_loinClothBack_c_MSTR [{shapeType=Capsule, length=1, radius=1}]
 dyn_loinClothBack_l_MSTR [{shapeType=Capsule, length=1, radius=1}]
 dyn_loinClothBack_r_MSTR [{shapeType=Capsule, length=1, radius=1}]
 pelvis [{shapeType=Capsule, length=16.3336, radius=8.3382}]
 thigh_l [{shapeType=Capsule, length=29.7498, radius=7.5675, center=<-18.4147 -0.6973 0.0002>, rotation=<rot 90 90 91.0202>}]
 thigh_r [{shapeType=Capsule, length=29.2498, radius=8.0675, center=<18.4147 -0.6973 0.0002>, rotation=<rot 90 8.8807 7.3201>}]
}
constraints [
 {
  bone1Name=thigh_l
  bone2Name=pelvis
  swing1LimitDegrees=45
  swing1Motion=Free
  swing2LimitDegrees=45
  swing2Motion=Free
  twistLimitDegrees=45
  twistMotion=Free
 }
 {
  bone1Name=thigh_r
  bone2Name=pelvis
  swing1LimitDegrees=45
  swing1Motion=Free
  swing2LimitDegrees=45
  swing2Motion=Free
  twistLimitDegrees=45
  twistMotion=Free
 }
 {bone1Name=dyn_loinClothBack_01_c, bone2Name=dyn_loinClothBack_c_MSTR, swing1LimitDegrees=30, swing2LimitDegrees=8, twistLimitDegrees=8}
 {bone1Name=dyn_loinClothBack_02_c, bone2Name=dyn_loinClothBack_01_c, swing1LimitDegrees=60, swing2LimitDegrees=10, twistLimitDegrees=8}
 {bone1Name=dyn_loinClothBack_03_c, bone2Name=dyn_loinClothBack_02_c, swing1LimitDegrees=60, swing2LimitDegrees=15, twistLimitDegrees=8}
 {bone1Name=dyn_loinClothBack_01_r, bone2Name=dyn_loinClothBack_r_MSTR, swing1LimitDegrees=30, swing2LimitDegrees=8, twistLimitDegrees=8}
 {bone1Name=dyn_loinClothBack_02_r, bone2Name=dyn_loinClothBack_01_r, swing1LimitDegrees=60, swing2LimitDegrees=10, twistLimitDegrees=8}
 {bone1Name=dyn_loinClothBack_03_r, bone2Name=dyn_loinClothBack_02_r, swing1LimitDegrees=60, swing2LimitDegrees=15, twistLimitDegrees=8}
 {bone1Name=dyn_loinClothBack_01_l, bone2Name=dyn_loinClothBack_l_MSTR, swing1LimitDegrees=30, swing2LimitDegrees=8, twistLimitDegrees=8}
 {bone1Name=dyn_loinClothBack_02_l, bone2Name=dyn_loinClothBack_01_l, swing1LimitDegrees=60, swing2LimitDegrees=10, twistLimitDegrees=8}
 {bone1Name=dyn_loinClothBack_03_l, bone2Name=dyn_loinClothBack_02_l, swing1LimitDegrees=60, swing2LimitDegrees=15, twistLimitDegrees=8}
 {
  bone1Name=dyn_loinClothBack_r_MSTR
  bone2Name=pelvis
  swing1LimitDegrees=45
  swing1Motion=Free
  swing2LimitDegrees=45
  swing2Motion=Free
  twistLimitDegrees=45
  twistMotion=Free
 }
 {
  bone1Name=dyn_loinClothBack_c_MSTR
  bone2Name=pelvis
  swing1LimitDegrees=45
  swing1Motion=Free
  swing2LimitDegrees=45
  swing2Motion=Free
  twistLimitDegrees=45
  twistMotion=Free
 }
 {
  bone1Name=dyn_loinClothBack_l_MSTR
  bone2Name=pelvis
  swing1LimitDegrees=45
  swing1Motion=Free
  swing2LimitDegrees=45
  swing2Motion=Free
  twistLimitDegrees=45
  twistMotion=Free
 }
 {bone1Name=dyn_loinClothBack_02_c, bone2Name=dyn_loinClothBack_02_l, swing1LimitDegrees=45, swing2LimitDegrees=45, twistLimitDegrees=45}
 {bone1Name=dyn_loinClothBack_02_c, bone2Name=dyn_loinClothBack_02_r, swing1LimitDegrees=45, swing2LimitDegrees=45, twistLimitDegrees=45}
 {bone1Name=dyn_loinClothBack_03_c, bone2Name=dyn_loinClothBack_03_l, swing1LimitDegrees=45, swing2LimitDegrees=45, twistLimitDegrees=45}
 {bone1Name=dyn_loinClothBack_03_c, bone2Name=dyn_loinClothBack_03_r, swing1LimitDegrees=45, swing2LimitDegrees=45, twistLimitDegrees=45}
]
