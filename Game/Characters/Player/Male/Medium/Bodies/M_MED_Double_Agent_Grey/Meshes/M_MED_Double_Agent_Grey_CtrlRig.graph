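# Control Rig for the Double Agent Grey medium male player body that makes front loincloth bones follow the legs.
# A chain of maintain-offset transform constraints drives left cloth bones fully from the left thigh twist bone.
# Right bones blend the right thigh twist with their own transforms, with weights varying along the chain.
# Center bones run last, blending the left thigh twist with the already-constrained right cloth bones.
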ControlRigBlueprint /Game/Characters/Player/Male/Medium/Bodies/M_MED_Double_Agent_Grey/Meshes/M_MED_Double_Agent_Grey_CtrlRig
graphs [
 RigVMFunctionLibrary {links [], nodes []}
 RigVMModel {nodes [
   RigUnit_BeginExecution = &RigUnit_BeginExecution()
   /Script/ControlRig.RigUnit_TransformConstraintPerItem(
    BaseItem=(Type=Bone)
    BaseTransform=(Rotation=(X=0,Y=0,Z=0,W=1),Translation=(X=0,Y=0,Z=0),Scale3D=(X=1,Y=1,Z=1))
    BaseTransformSpace=GlobalSpace
    Item=(Type=Bone,Name="dyn_loinClothFront_l_MSTR")
    Targets=((Transform=(Rotation=(X=0,Y=0,Z=0,W=1),Translation=(X=0,Y=0,Z=0),Scale3D=(X=1,Y=1,Z=1)),Weight=1,bMaintainOffset=True,Filter=(TranslationFilter=(bX=True,bY=True,bZ=True),RotationFilter=(bX=True,bY=True,bZ=True),ScaleFilter=(bX=True,bY=True,bZ=True))))
    bUseInitialTransforms=True
    Targets.0.Transform=&RigUnit_GetTransform(Item=(Type=Bone,Name="thigh_twist_01_l")).Transform
    ExecutePin=@RigUnit_BeginExecution.ExecutePin
   ) {is=struct RigVMUnitNode}
   > &RigUnit_TransformConstraintPerItem(
    BaseItem=(Type=Bone)
    BaseTransform=(Rotation=(X=0,Y=0,Z=0,W=1),Translation=(X=0,Y=0,Z=0),Scale3D=(X=1,Y=1,Z=1))
    BaseTransformSpace=GlobalSpace
    Item=(Type=Bone,Name="dyn_loinClothFront_01_l")
    Targets=((Transform=(Rotation=(X=0,Y=0,Z=0,W=1),Translation=(X=0,Y=0,Z=0),Scale3D=(X=1,Y=1,Z=1)),Weight=1,bMaintainOffset=True,Filter=(TranslationFilter=(bX=True,bY=True,bZ=True),RotationFilter=(bX=True,bY=True,bZ=True),ScaleFilter=(bX=True,bY=True,bZ=True))))
    bUseInitialTransforms=True
    Targets.0.Transform=&RigUnit_GetTransform(Item=(Type=Bone,Name="thigh_twist_01_l")).Transform
   )
   > &RigUnit_TransformConstraintPerItem(
    BaseItem=(Type=Bone)
    BaseTransform=(Rotation=(X=0,Y=0,Z=0,W=1),Translation=(X=0,Y=0,Z=0),Scale3D=(X=1,Y=1,Z=1))
    BaseTransformSpace=GlobalSpace
    Item=(Type=Bone,Name="dyn_loinClothFront_02_l")
    Targets=((Transform=(Rotation=(X=0,Y=0,Z=0,W=1),Translation=(X=0,Y=0,Z=0),Scale3D=(X=1,Y=1,Z=1)),Weight=1,bMaintainOffset=True,Filter=(TranslationFilter=(bX=True,bY=True,bZ=True),RotationFilter=(bX=True,bY=True,bZ=True),ScaleFilter=(bX=True,bY=True,bZ=True))))
    bUseInitialTransforms=True
    Targets.0.Transform=&RigUnit_GetTransform(Item=(Type=Bone,Name="thigh_twist_01_l")).Transform
   )
   > &RigUnit_TransformConstraintPerItem(
    BaseItem=(Type=Bone)
    BaseTransform=(Rotation=(X=0,Y=0,Z=0,W=1),Translation=(X=0,Y=0,Z=0),Scale3D=(X=1,Y=1,Z=1))
    BaseTransformSpace=GlobalSpace
    Item=(Type=Bone,Name="dyn_loinClothFront_03_l")
    Targets=((Transform=(Rotation=(X=0,Y=0,Z=0,W=1),Translation=(X=0,Y=0,Z=0),Scale3D=(X=1,Y=1,Z=1)),Weight=1,bMaintainOffset=True,Filter=(TranslationFilter=(bX=True,bY=True,bZ=True),RotationFilter=(bX=True,bY=True,bZ=True),ScaleFilter=(bX=True,bY=True,bZ=True))))
    bUseInitialTransforms=True
    Targets.0.Transform=&RigUnit_GetTransform(Item=(Type=Bone,Name="thigh_twist_01_l")).Transform
   )
   > &RigUnit_TransformConstraintPerItem(
    BaseItem=(Type=Bone)
    BaseTransform=(Rotation=(X=0,Y=0,Z=0,W=1),Translation=(X=0,Y=0,Z=0),Scale3D=(X=1,Y=1,Z=1))
    BaseTransformSpace=GlobalSpace
    Item=(Type=Bone,Name="dyn_loinClothFront_r_MSTR")
    Targets="((Transform=(Rotation=(X=0,Y=0,Z=0,W=1),Translation=(X=0,Y=0,Z=0),Scale3D=(X=1,Y=1,Z=1)),Weight=0.5,bMaintainOffset=True,Filter=(TranslationFilter=(bX=True,bY=True,bZ=True),RotationFilter=(bX=True,bY=True,bZ=True),ScaleFilter=(bX=True,bY=True,bZ=True))),(T ...(truncated)"
    bUseInitialTransforms=True
    Targets.0.Transform=&RigUnit_GetTransform(Item=(Type=Bone,Name="thigh_twist_01_r")).Transform
    Targets.1.Transform=&RigUnit_GetTransform(Item=(Type=Bone,Name="dyn_loinClothFront_r_MSTR")).Transform
   )
   > &RigUnit_TransformConstraintPerItem(
    BaseItem=(Type=Bone)
    BaseTransform=(Rotation=(X=0,Y=0,Z=0,W=1),Translation=(X=0,Y=0,Z=0),Scale3D=(X=1,Y=1,Z=1))
    BaseTransformSpace=GlobalSpace
    Item=(Type=Bone,Name="dyn_loinClothFront_01_r")
    Targets="((Transform=(Rotation=(X=0,Y=0,Z=0,W=1),Translation=(X=0,Y=0,Z=0),Scale3D=(X=1,Y=1,Z=1)),Weight=0.75,bMaintainOffset=True,Filter=(TranslationFilter=(bX=True,bY=True,bZ=True),RotationFilter=(bX=True,bY=True,bZ=True),ScaleFilter=(bX=True,bY=True,bZ=True))),( ...(truncated)"
    bUseInitialTransforms=True
    Targets.0.Transform=&RigUnit_GetTransform(Item=(Type=Bone,Name="thigh_twist_01_r")).Transform
    Targets.1.Transform=&RigUnit_GetTransform(Item=(Type=Bone,Name="dyn_loinClothFront_01_r")).Transform
   )
   > &RigUnit_TransformConstraintPerItem(
    BaseItem=(Type=Bone)
    BaseTransform=(Rotation=(X=0,Y=0,Z=0,W=1),Translation=(X=0,Y=0,Z=0),Scale3D=(X=1,Y=1,Z=1))
    BaseTransformSpace=GlobalSpace
    Item=(Type=Bone,Name="dyn_loinClothFront_02_r")
    Targets="((Transform=(Rotation=(X=0,Y=0,Z=0,W=1),Translation=(X=0,Y=0,Z=0),Scale3D=(X=1,Y=1,Z=1)),Weight=0.75,bMaintainOffset=True,Filter=(TranslationFilter=(bX=True,bY=True,bZ=True),RotationFilter=(bX=True,bY=True,bZ=True),ScaleFilter=(bX=True,bY=True,bZ=True))),( ...(truncated)"
    bUseInitialTransforms=True
    Targets.0.Transform=&RigUnit_GetTransform(Item=(Type=Bone,Name="thigh_twist_01_r")).Transform
    Targets.1.Transform=&RigUnit_GetTransform(Item=(Type=Bone,Name="dyn_loinClothFront_02_r")).Transform
   )
   > &RigUnit_TransformConstraintPerItem(
    BaseItem=(Type=Bone)
    BaseTransform=(Rotation=(X=0,Y=0,Z=0,W=1),Translation=(X=0,Y=0,Z=0),Scale3D=(X=1,Y=1,Z=1))
    BaseTransformSpace=GlobalSpace
    Item=(Type=Bone,Name="dyn_loinClothFront_03_r")
    Targets="((Transform=(Rotation=(X=0,Y=0,Z=0,W=1),Translation=(X=0,Y=0,Z=0),Scale3D=(X=1,Y=1,Z=1)),Weight=0.5,bMaintainOffset=True,Filter=(TranslationFilter=(bX=True,bY=True,bZ=True),RotationFilter=(bX=True,bY=True,bZ=True),ScaleFilter=(bX=True,bY=True,bZ=True))),(T ...(truncated)"
    bUseInitialTransforms=True
    Targets.0.Transform=&RigUnit_GetTransform(Item=(Type=Bone,Name="thigh_twist_01_r")).Transform
    Targets.1.Transform=&RigUnit_GetTransform(Item=(Type=Bone,Name="dyn_loinClothFront_03_r")).Transform
   )
   > &RigUnit_TransformConstraintPerItem(
    BaseItem=(Type=Bone)
    BaseTransform=(Rotation=(X=0,Y=0,Z=0,W=1),Translation=(X=0,Y=0,Z=0),Scale3D=(X=1,Y=1,Z=1))
    BaseTransformSpace=GlobalSpace
    Item=(Type=Bone,Name="dyn_loinClothFront_c_MSTR")
    Targets="((Transform=(Rotation=(X=0,Y=0,Z=0,W=1),Translation=(X=0,Y=0,Z=0),Scale3D=(X=1,Y=1,Z=1)),Weight=0.5,bMaintainOffset=True,Filter=(TranslationFilter=(bX=True,bY=True,bZ=True),RotationFilter=(bX=True,bY=True,bZ=True),ScaleFilter=(bX=True,bY=True,bZ=True))),(T ...(truncated)"
    bUseInitialTransforms=True
    Targets.0.Transform=&RigUnit_GetTransform(Item=(Type=Bone,Name="thigh_twist_01_l")).Transform
    Targets.1.Transform=&RigUnit_GetTransform(Item=(Type=Bone,Name="dyn_loinClothFront_r_MSTR")).Transform
   )
   > &RigUnit_TransformConstraintPerItem(
    BaseItem=(Type=Bone)
    BaseTransform=(Rotation=(X=0,Y=0,Z=0,W=1),Translation=(X=0,Y=0,Z=0),Scale3D=(X=1,Y=1,Z=1))
    BaseTransformSpace=GlobalSpace
    Item=(Type=Bone,Name="dyn_loinClothFront_01_c")
    Targets="((Transform=(Rotation=(X=0,Y=0,Z=0,W=1),Translation=(X=0,Y=0,Z=0),Scale3D=(X=1,Y=1,Z=1)),Weight=0.5,bMaintainOffset=True,Filter=(TranslationFilter=(bX=True,bY=True,bZ=True),RotationFilter=(bX=True,bY=True,bZ=True),ScaleFilter=(bX=True,bY=True,bZ=True))),(T ...(truncated)"
    bUseInitialTransforms=True
    Targets.0.Transform=&RigUnit_GetTransform(Item=(Type=Bone,Name="thigh_twist_01_l")).Transform
    Targets.1.Transform=&RigUnit_GetTransform(Item=(Type=Bone,Name="dyn_loinClothFront_01_r")).Transform
   )
   > &RigUnit_TransformConstraintPerItem(
    BaseItem=(Type=Bone)
    BaseTransform=(Rotation=(X=0,Y=0,Z=0,W=1),Translation=(X=0,Y=0,Z=0),Scale3D=(X=1,Y=1,Z=1))
    BaseTransformSpace=GlobalSpace
    Item=(Type=Bone,Name="dyn_loinClothFront_02_c")
    Targets="((Transform=(Rotation=(X=0,Y=0,Z=0,W=1),Translation=(X=0,Y=0,Z=0),Scale3D=(X=1,Y=1,Z=1)),Weight=0.5,bMaintainOffset=True,Filter=(TranslationFilter=(bX=True,bY=True,bZ=True),RotationFilter=(bX=True,bY=True,bZ=True),ScaleFilter=(bX=True,bY=True,bZ=True))),(T ...(truncated)"
    bUseInitialTransforms=True
    Targets.0.Transform=&RigUnit_GetTransform(Item=(Type=Bone,Name="thigh_twist_01_l")).Transform
    Targets.1.Transform=&RigUnit_GetTransform(Item=(Type=Bone,Name="dyn_loinClothFront_02_r")).Transform
   )
   > &RigUnit_TransformConstraintPerItem(
    BaseItem=(Type=Bone)
    BaseTransform=(Rotation=(X=0,Y=0,Z=0,W=1),Translation=(X=0,Y=0,Z=0),Scale3D=(X=1,Y=1,Z=1))
    BaseTransformSpace=GlobalSpace
    Item=(Type=Bone,Name="dyn_loinClothFront_03_c")
    Targets="((Transform=(Rotation=(X=0,Y=0,Z=0,W=1),Translation=(X=0,Y=0,Z=0),Scale3D=(X=1,Y=1,Z=1)),Weight=0.5,bMaintainOffset=True,Filter=(TranslationFilter=(bX=True,bY=True,bZ=True),RotationFilter=(bX=True,bY=True,bZ=True),ScaleFilter=(bX=True,bY=True,bZ=True))),(T ...(truncated)"
    bUseInitialTransforms=True
    Targets.0.Transform=&RigUnit_GetTransform(Item=(Type=Bone,Name="thigh_twist_01_l")).Transform
    Targets.1.Transform=&RigUnit_GetTransform(Item=(Type=Bone,Name="dyn_loinClothFront_03_r")).Transform
   )
  ]}
]
hierarchy [
 root {loc=<0 0 0>, rot=<0 0 0 1>}
 pelvis {parent=root, loc=<0 -0.7071 93.1191>, rot=<0.0127 0.707 -0.0127 -0.707>}
 spine_01 {parent=pelvis, loc=<6.7494 -0.6388 0>, rot=<0 0 0.034 -0.9994>}
 spine_02 {parent=spine_01, loc=<7.2518 -0.035 0>, rot=<0 0 0.0238 -0.9997>}
 spine_03 {parent=spine_02, loc=<6.0104 0 0>, rot=<0 0 0.0149 -0.9999>}
 spine_04 {parent=spine_03, loc=<8.1024 0 0>, rot=<0 0 -0.0132 -0.9999>}
 spine_05 {parent=spine_04, loc=<10.0146 0 0>, rot=<0 0 -0.0015 -1>}
 clavicle_l {parent=spine_05, loc=<4.3575 -0.1466 -2.3223>, rot=<0.0465 -0.7289 -0.009 -0.6829>}
 upperarm_l {parent=clavicle_l, loc=<15.5111 0 0>, rot=<0.0226 -0.3234 -0.0088 -0.9459>}
 lowerarm_l {parent=upperarm_l, loc=<26.6514 0 0>, rot=<0.0391 -0.0099 -0.2453 -0.9686>}
 hand_l {parent=lowerarm_l, loc=<22.3491 0 0>, rot=<-0.7081 -0.0507 0.0975 -0.6975>}
 index_metacarpal_l {parent=hand_l, loc=<1.6535 0.0183 -2.96>, rot=<0.0158 -0.0124 -0.0173 -0.9996>}
 index_01_l {parent=index_metacarpal_l, loc=<8.3008 0.0357 -0.0524>, rot=<-0.0044 0.0485 0.0924 -0.9945>}
 index_02_l {parent=index_01_l, loc=<3.3587 -0.3004 0.0452>, rot=<-0.002 0.0027 0.3225 -0.9466>}
 index_03_l {parent=index_02_l, loc=<2.5578 -0.123 -0.0252>, rot=<-0.002 0.0067 0.2558 -0.9667>}
 middle_metacarpal_l {parent=hand_l, loc=<2.1856 -0.1725 -0.5984>, rot=<0.006 0.0168 -0.0329 -0.9993>}
 middle_01_l {parent=middle_metacarpal_l, loc=<7.3275 -0.2447 0.0129>, rot=<-0.0208 0.0042 0.0858 -0.9961>}
 middle_02_l {parent=middle_01_l, loc=<4.0123 0.1401 0.1175>, rot=<-0.0218 0.0258 0.3512 -0.9357>}
 middle_03_l {parent=middle_02_l, loc=<2.6077 0.0534 0.1373>, rot=<0.0082 0.0188 0.1708 -0.9851>}
 pinky_metacarpal_l {parent=hand_l, loc=<1.7616 -0.9753 3.3814>, rot=<-0.0397 0.0814 -0.069 -0.9935>}
 pinky_01_l {parent=pinky_metacarpal_l, loc=<7.1709 -0.4831 -0.1094>, rot=<-0.0654 -0.0035 0.2242 -0.9723>}
 pinky_02_l {parent=pinky_01_l, loc=<2.8274 0.0893 -0.1043>, rot=<0.0037 0.0011 0.2854 -0.9584>}
 pinky_03_l {parent=pinky_02_l, loc=<2.2046 0.0307 -0.0078>, rot=<-0.0002 0.0098 0.1445 -0.9895>}
 ring_metacarpal_l {parent=hand_l, loc=<2.2074 -0.3625 1.3638>, rot=<-0.0284 0.0722 -0.0458 -0.9959>}
 ring_01_l {parent=ring_metacarpal_l, loc=<7.0295 -0.6197 -0.1446>, rot=<-0.0489 -0.0252 0.1892 -0.9804>}
 ring_02_l {parent=ring_01_l, loc=<3.6174 0.4647 -0.1027>, rot=<0 0 0.2806 -0.9598>}
 ring_03_l {parent=ring_02_l, loc=<2.4152 0.114 0.0199>, rot=<0 0 0.1189 -0.9929>}
 thumb_01_l {parent=hand_l, loc=<1.3118 -0.3191 -2.459>, rot=<0.6546 -0.4042 0.0664 -0.6354>}
 thumb_02_l {parent=thumb_01_l, loc=<5.9336 -0.0448 0.0401>, rot=<-0.0037 0.0213 0.298 -0.9543>}
 thumb_03_l {parent=thumb_02_l, loc=<2.2223 0.3682 -0.0907>, rot=<0.0118 -0.013 0.0066 -0.9998>}
 lowerarm_twist_01_l {parent=lowerarm_l, loc=<17.8549 0 0>, rot=<0 0 0 -1>}
 lowerarm_twist_02_l {parent=lowerarm_l, loc=<9.6199 0 0>, rot=<0 0 0 -1>}
 dyn_L_sleeve_jnt_A {parent=lowerarm_l, loc=<-0.0439 -0.4404 4.8097>, rot=<0.1199 0.6662 0.1992 -0.7086>}
 dyn_L_sleeve_jnt_B {parent=lowerarm_l, loc=<1.3184 0.5471 -4.0159>, rot=<-0.1989 -0.7071 0.1204 -0.6678>}
 upperarm_twist_01_l {parent=upperarm_l, loc=<0.0041 0 0>, rot=<0 0 0 -1>}
 upperarm_twist_02_l {parent=upperarm_l, loc=<13.7941 0 0>, rot=<0 0 0 -1>}
 clavicle_r {parent=spine_05, loc=<4.3575 -0.1466 2.3223>, rot=<0.7289 0.0465 -0.6829 0.009>}
 upperarm_r {parent=clavicle_r, loc=<-15.5111 0 0>, rot=<0.0226 -0.3234 -0.0088 -0.9459>}
 lowerarm_r {parent=upperarm_r, loc=<-26.6515 0 0>, rot=<0.0391 -0.0099 -0.2453 -0.9686>}
 hand_r {parent=lowerarm_r, loc=<-22.3492 0 0>, rot=<-0.7081 -0.0507 0.0975 -0.6975>}
 index_metacarpal_r {parent=hand_r, loc=<-1.6534 -0.0181 2.9601>, rot=<0.0158 -0.0124 -0.0173 -0.9996>}
 index_01_r {parent=index_metacarpal_r, loc=<-8.3008 -0.0357 0.0524>, rot=<-0.0044 0.0485 0.0924 -0.9945>}
 index_02_r {parent=index_01_r, loc=<-3.3589 0.3003 -0.0453>, rot=<-0.002 0.0027 0.3225 -0.9466>}
 index_03_r {parent=index_02_r, loc=<-2.5578 0.1231 0.0251>, rot=<-0.002 0.0067 0.2558 -0.9667>}
 middle_metacarpal_r {parent=hand_r, loc=<-2.1856 0.1726 0.5984>, rot=<0.006 0.0168 -0.0329 -0.9993>}
 middle_01_r {parent=middle_metacarpal_r, loc=<-7.3273 0.2447 -0.0129>, rot=<-0.0208 0.0042 0.0858 -0.9961>}
 middle_02_r {parent=middle_01_r, loc=<-4.0142 -0.1411 -0.07>, rot=<-0.0217 0.0259 0.3512 -0.9357>}
 middle_03_r {parent=middle_02_r, loc=<-2.6074 -0.0534 -0.1312>, rot=<0.0083 0.0188 0.1708 -0.9851>}
 pinky_metacarpal_r {parent=hand_r, loc=<-1.7617 0.9753 -3.3814>, rot=<-0.0397 0.0814 -0.069 -0.9935>}
 pinky_01_r {parent=pinky_metacarpal_r, loc=<-7.1707 0.4832 0.1094>, rot=<-0.0654 -0.0035 0.2242 -0.9723>}
 pinky_02_r {parent=pinky_01_r, loc=<-2.8274 -0.0892 0.1043>, rot=<0.0037 0.0011 0.2854 -0.9584>}
 pinky_03_r {parent=pinky_02_r, loc=<-2.2045 -0.031 0.0078>, rot=<-0.0002 0.0098 0.1445 -0.9895>}
 ring_metacarpal_r {parent=hand_r, loc=<-2.2073 0.3627 -1.3638>, rot=<-0.0284 0.0722 -0.0458 -0.9959>}
 ring_01_r {parent=ring_metacarpal_r, loc=<-7.0297 0.6197 0.1446>, rot=<-0.0489 -0.0252 0.1892 -0.9804>}
 ring_02_r {parent=ring_01_r, loc=<-3.6171 -0.4648 0.1027>, rot=<0 0 0.2806 -0.9598>}
 ring_03_r {parent=ring_02_r, loc=<-2.4153 -0.1141 -0.0199>, rot=<0 0 0.1189 -0.9929>}
 thumb_01_r {parent=hand_r, loc=<-1.3119 0.3192 2.459>, rot=<0.6546 -0.4042 0.0664 -0.6354>}
 thumb_02_r {parent=thumb_01_r, loc=<-5.9336 0.0447 -0.0401>, rot=<-0.0037 0.0213 0.298 -0.9543>}
 thumb_03_r {parent=thumb_02_r, loc=<-2.2223 -0.3682 0.0907>, rot=<0.0118 -0.013 0.0066 -0.9998>}
 lowerarm_twist_01_r {parent=lowerarm_r, loc=<-17.855 0 0>, rot=<0 0 0 1>}
 lowerarm_twist_02_r {parent=lowerarm_r, loc=<-9.6199 0 0>, rot=<0 0 0 1>}
 dyn_R_sleeve_jnt_A {parent=lowerarm_twist_02_r, loc=<1.9979 0.6929 -5.204>, rot=<-0.7039 -0.0398 0.7081 0.0377>}
 dyn_R_sleeve_jnt_B {parent=lowerarm_twist_02_r, loc=<-2.7664 -0.0184 3.8553>, rot=<0.708 -0.0155 0.7032 -0.0637>}
 upperarm_twist_01_r {parent=upperarm_r, loc=<-0.0041 0 0>, rot=<0 0 0 1>}
 upperarm_twist_02_r {parent=upperarm_r, loc=<-13.7942 0 0>, rot=<0 0 0 1>}
 dyn_L_shirtFlap_jnt {parent=spine_01, loc=<0.974 0.4201 -13.3725>, rot=<0.0114 -0.707 -0.0114 -0.707>}
 dyn_R_shirtFlap_jnt {parent=spine_01, loc=<0.883 1.0951 13.305>, rot=<0.707 0.0114 0.707 -0.0114>}
 thigh_l {parent=pelvis, loc=<-2.5693 -0.0109 -7.7501>, rot=<0.0012 0.0572 0.0155 -0.9982>}
 calf_l {parent=thigh_l, loc=<-39.0484 0 0>, rot=<0.0006 -0.0161 -0.0391 -0.9991>}
 calf_twist_01_l {parent=calf_l, loc=<-35.813 -0.0967 0.0397>, rot=<0.0237 0.0095 -0.0035 -0.9997>}
 calf_twist_02_l {parent=calf_l, loc=<-18.8064 -0.0967 0.0397>, rot=<0.0237 0.0095 -0.0035 -0.9997>}
 foot_l {parent=calf_l, loc=<-42.561 0.1077 0.4292>, rot=<0.0418 -0.0455 0.0343 -0.9975>}
 ball_l {parent=foot_l, loc=<-6.1599 13.8264 0.6806>, rot=<0.0001 -0.0001 -0.7063 -0.7079>}
 thigh_twist_01_l {parent=thigh_l, loc=<-21.6303 0 0>, rot=<0.0013 0 0 -1>}
 dyn_L_frontSkirt_jnt_MASTER {parent=thigh_l, loc=<-2.9938 8.7442 -0.2224>, rot=<-0.0849 -0.9959 0.0129 -0.0279>}
 dyn_L_frontSkirt_jnt_1 {parent=dyn_L_frontSkirt_jnt_MASTER, loc=<0 0 0>, rot=<0 0 0 1>}
 dyn_L_frontSkirt_jnt_2 {parent=dyn_L_frontSkirt_jnt_1, loc=<5.3564 0 0>, rot=<-0.002 -0.0122 0.0073 -0.9999>}
 dyn_L_frontSkirt_jnt_3 {parent=dyn_L_frontSkirt_jnt_2, loc=<5.4164 0 0>, rot=<0 0 0 -1>}
 dyn_L_strap_jnt_1 {parent=thigh_l, loc=<0.1454 1.1324 -7.8269>, rot=<-0.0098 -0.9985 -0.0018 -0.053>}
 dyn_L_strap_jnt_2 {parent=dyn_L_strap_jnt_1, loc=<12.4891 -0.5363 -0.0941>, rot=<-0.0015 -0.0156 0.015 -0.9998>}
 dyn_L_strap_jnt_3 {parent=dyn_L_strap_jnt_2, loc=<12.3692 0.0533 -0.1072>, rot=<-0.0001 0.0101 -0.002 -0.9999>}
 thigh_r {parent=pelvis, loc=<-2.5691 -0.0109 7.7501>, rot=<0.0572 -0.0012 0.9982 0.0155>}
 calf_r {parent=thigh_r, loc=<39.0487 0 0>, rot=<0.0006 -0.0161 -0.0391 -0.9991>}
 calf_twist_01_r {parent=calf_r, loc=<35.8132 0.0967 -0.0397>, rot=<0.0237 0.0095 -0.0035 -0.9997>}
 calf_twist_02_r {parent=calf_r, loc=<18.8064 0.0967 -0.0397>, rot=<0.0237 0.0095 -0.0035 -0.9997>}
 foot_r {parent=calf_r, loc=<42.5612 -0.1077 -0.4292>, rot=<0.0418 -0.0455 0.0343 -0.9975>}
 ball_r {parent=foot_r, loc=<6.1598 -13.8264 -0.6806>, rot=<0.0001 -0.0001 -0.7063 -0.7079>}
 thigh_twist_01_r {parent=thigh_r, loc=<21.6303 0 0>, rot=<0.0013 0 0 -1>}
 dyn_R_strap_jnt_1 {parent=thigh_r, loc=<-0.1452 -1.1324 7.8269>, rot=<0.9985 -0.0098 0.053 -0.0018>}
 dyn_R_strap_jnt_2 {parent=dyn_R_strap_jnt_1, loc=<12.4891 -0.5363 0.0941>, rot=<0.0015 0.0156 0.015 -0.9998>}
 dyn_R_strap_jnt_3 {parent=dyn_R_strap_jnt_2, loc=<12.3693 0.0533 0.1071>, rot=<0.0001 -0.0101 -0.002 -0.9999>}
 dyn_mid_backSkirt_jnt_MASTER {parent=pelvis, loc=<-5.0164 -11.0104 -0.1168>, rot=<0.0594 -0.9982 0 0>}
 dyn_mid_backSkirt_jnt_1 {parent=dyn_mid_backSkirt_jnt_MASTER, loc=<0 0 0>, rot=<0 0 0 -1>}
 dyn_back_strap_jnt_2 {parent=dyn_mid_backSkirt_jnt_1, loc=<8.2868 -0.1837 -0.1212>, rot=<-0.0001 0.0004 0.0108 -0.9999>}
 dyn_back_strap_jnt_3 {parent=dyn_back_strap_jnt_2, loc=<8.4551 -0.005 -0.0058>, rot=<0 0 0 1>}
 dyn_back_strap_jnt_4 {parent=dyn_back_strap_jnt_3, loc=<10.2408 0.1974 -0.0045>, rot=<0 -0.0002 -0.013 -0.9999>}
 dyn_back_strap_jnt_5 {parent=dyn_back_strap_jnt_4, loc=<11.2156 -0.0889 -0.0007>, rot=<0 0.0002 0.0045 -1>}
 dyn_loinClothBack_l_MSTR {parent=pelvis, loc=<-2.3515 -9.1573 -10.0291>, rot=<0.0453 -0.9966 -0.0581 -0.0366>}
 dyn_loinClothBack_01_l {parent=dyn_loinClothBack_l_MSTR, loc=<0 0 0>, rot=<0 0 0 1>}
 dyn_loinClothBack_02_l {parent=dyn_loinClothBack_01_l, loc=<9.5517 0 0>, rot=<-0.014 -0.0047 -0.048 -0.9987>}
 dyn_loinClothBack_03_l {parent=dyn_loinClothBack_02_l, loc=<6.5193 0 0>, rot=<0.0091 0.0019 -0.0127 -0.9999>}
 dyn_loinClothBack_r_MSTR {parent=pelvis, loc=<-2.3547 -9.1566 9.5888>, rot=<0.0438 -0.9959 0.0786 0.0047>}
 dyn_loinClothBack_01_r {parent=dyn_loinClothBack_r_MSTR, loc=<0 0 0>, rot=<0 0 0 1>}
 dyn_loinClothBack_02_r {parent=dyn_loinClothBack_01_r, loc=<9.5025 0 0>, rot=<-0.0422 -0.012 -0.0453 -0.998>}
 dyn_loinClothBack_03_r {parent=dyn_loinClothBack_02_r, loc=<6.3252 0 0>, rot=<-0.0271 -0.0184 -0.0213 -0.9992>}
 dyn_loinClothBack_c_MSTR {parent=pelvis, loc=<-2.3546 -10.0092 -0.4535>, rot=<0.0253 -0.9994 -0.0222 -0.0101>}
 dyn_loinClothBack_01_c {parent=dyn_loinClothBack_c_MSTR, loc=<0 0 0>, rot=<0 0 0 -1>}
 dyn_loinClothBack_02_c {parent=dyn_loinClothBack_01_c, loc=<9.2614 0 0>, rot=<0.0093 0.0004 -0.0328 -0.9994>}
 dyn_loinClothBack_03_c {parent=dyn_loinClothBack_02_c, loc=<6.5135 0 0>, rot=<-0.0022 -0.0003 -0.0374 -0.9993>}
 dyn_loinClothFront_l_MSTR {parent=pelvis, loc=<-1.5036 7.7809 -9.915>, rot=<-0.0104 0.1116 0.9855 0.1274>}
 dyn_loinClothFront_01_l {parent=dyn_loinClothFront_l_MSTR, loc=<0 0 0>, rot=<0 0 0 1>}
 dyn_loinClothFront_02_l {parent=dyn_loinClothFront_01_l, loc=<9.5025 0 0>, rot=<-0.0425 -0.0109 -0.0698 -0.9966>}
 dyn_loinClothFront_03_l {parent=dyn_loinClothFront_02_l, loc=<6.3252 0 0>, rot=<-0.0271 -0.0184 -0.0213 -0.9992>}
 dyn_loinClothFront_r_MSTR {parent=pelvis, loc=<-1.5287 7.6931 9.6348>, rot=<0.0195 -0.0279 0.9903 0.1345>}
 dyn_loinClothFront_01_r {parent=dyn_loinClothFront_r_MSTR, loc=<0 0 0>, rot=<0 0 0 1>}
 dyn_loinClothFront_02_r {parent=dyn_loinClothFront_01_r, loc=<9.5517 0 0>, rot=<-0.0141 -0.0043 -0.0725 -0.9973>}
 dyn_loinClothFront_03_r {parent=dyn_loinClothFront_02_r, loc=<6.5193 0 0>, rot=<0.0091 0.0019 -0.0127 -0.9999>}
 dyn_loinClothFront_c_MSTR {parent=pelvis, loc=<-2.1519 9.4708 0.0151>, rot=<-0.0034 0.0103 0.994 0.1086>}
 dyn_loinClothFront_01_c {parent=dyn_loinClothFront_c_MSTR, loc=<0 0 0>, rot=<0 0 0 1>}
 dyn_loinClothFront_02_c {parent=dyn_loinClothFront_01_c, loc=<9.2614 0 0>, rot=<0.0093 0.0004 -0.0328 -0.9994>}
 dyn_loinClothFront_03_c {parent=dyn_loinClothFront_02_c, loc=<6.5135 0 0>, rot=<-0.0022 -0.0003 -0.0374 -0.9993>}
]
