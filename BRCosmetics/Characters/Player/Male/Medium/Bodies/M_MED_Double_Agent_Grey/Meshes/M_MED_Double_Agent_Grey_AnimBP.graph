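# Post-process-style cosmetic AnimBP for the Double Agent Grey body, adding secondary motion to sleeves, loincloth and straps.
# It copies the parent mesh pose, then layers AnimDynamics, thigh constraints and RigidBody simulation, blending two loincloth setups by crouch state.
# bRigidBodyEnabled switches between RigidBody and Trail controllers before final corrective bone offsets.
# The event graph sets front-end/emote alphas and gravity overrides through a deprecated state function.
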AnimBlueprint /BRCosmetics/Characters/Player/Male/Medium/Bodies/M_MED_Double_Agent_Grey/Meshes/M_MED_Double_Agent_Grey_AnimBP
target_skeleton=/Game/Characters/Player/Male/Medium/Bodies/M_MED_Double_Agent_Grey/Meshes/M_MED_Double_Agent_Grey_Skeleton
graphs [
 AnimGraph {nodes [
   n1 = Anim Dynamics - Bone: dyn_L_shirtFlap_jnt(Alpha=1, ExternalForce="0,0,0") {
    bGravityOverrideInSimSpace=true
    gravityScale=1.4
    linearDampingOverride=1
    linearSpringConstant=50
    alphaBoolBlend {blendOption=Linear}
    boundBone {boneName=dyn_L_shirtFlap_jnt}
    componentAppliedLinearAccClamp=<4000 4000 4000>
    componentLinearAccScale=<3 3 3>
    componentLinearVelScale=<0.1 0.1 0.1>
    physicsBodyDefinitions [{
      sphereCollisionRadius=0
      boundBone {boneName=dyn_L_shirtFlap_jnt}
      boxExtents=<1 1 1>
      constraintSetup {angularLimitsMax=<0 0 0>, angularLimitsMin=<0 0 0>, angularTarget=<0 0 0>, linearAxesMax=<1 1 5>, linearAxesMin=<-1 -1 -1>}
      localJointOffset=<0 0 0>
     }]
    relativeSpaceBone {boneName=spine_01}
   }
   Anim Dynamics - Bone: dyn_R_shirtFlap_jnt(ComponentPose=@n1, Alpha=1, ExternalForce="0,0,0") {
    bGravityOverrideInSimSpace=true
    gravityScale=1.4
    linearDampingOverride=1
    linearSpringConstant=50
    alphaBoolBlend {blendOption=Linear}
    boundBone {boneName=dyn_R_shirtFlap_jnt}
    componentAppliedLinearAccClamp=<4000 4000 4000>
    componentLinearAccScale=<3 3 3>
    componentLinearVelScale=<0.1 0.1 0.1>
    physicsBodyDefinitions [{
      sphereCollisionRadius=0
      boundBone {boneName=dyn_R_shirtFlap_jnt}
      boxExtents=<1 1 1>
      constraintSetup {angularLimitsMax=<0 0 0>, angularLimitsMin=<0 0 0>, angularTarget=<0 0 0>, linearAxesMax=<1 1 5>, linearAxesMin=<-1 -1 -1>}
      localJointOffset=<0 0 0>
     }]
    relativeSpaceBone {boneName=spine_01}
   }
   Control Rig() {
    alphaBoolBlend {blendOption=Linear}
    controlRigAssetReference {blueprintRigClass=/Game/Characters/Player/Male/Medium/Bodies/M_MED_Double_Agent_Grey/Meshes/M_MED_Double_Agent_Grey_CtrlRig.M_MED_Double_Agent_Grey_CtrlRig_C}
   }
   "Transform (Modify) Bone - Bone: thigh_r"(Alpha=0, Translation="0,0,0", Rotation="0,90,0", Scale="1,1,1") {
    rotationMode=BMM_Additive
    rotationSpace=BCS_BoneSpace
    alphaBoolBlend {blendOption=Linear}
    boneToModify {boneName=thigh_r}
    rotation {yaw=90}
   }
   n2 = RigidBody!(ComponentPose=Local To Component(LocalPose=Use cached pose 'RBANenabled'()), Alpha, OverrideWorldGravity:Vector="0,0,-980", ExternalForce) -> (Pose) {
    bClampLinearTranslationLimitToRefPose=true
    bOverrideWorldGravity=true
    overridePhysicsAsset=&/M_MED_Double_Agent_Grey_Physics
    simulationSpace=BaseBoneSpace
    alphaBoolBlend {blendOption=Linear}
    baseBoneRef {boneName=root}
    componentAppliedLinearAccClamp=<4000 4000 4000>
    componentLinearAccScale=<0.3 0.3 0.3>
    componentLinearVelScale=<0.4 0.4 0.4>
    overrideWorldGravity {z=-980}
   }
   n3 = Component To Local(ComponentPose=@n2)
   n4 = Trail controller - Bone: dyn_back_strap_jnt_7!(
    ComponentPose=Local To Component(LocalPose=Use cached pose 'RBANenabled'())
    Alpha=1
    RelaxationSpeedScale:float=2
    FakeVelocity:Vector="0,0,0"
   ) -> (Pose) {
    bActorSpaceFakeVel=true
    bShowTrailLocation=true
    bUsePlanarLimit=true
    chainLength=5
    debugLifeTime=0.1
    maxDeltaTime=0.0333
    alphaScaleBiasClamp {bInterpResult=true, interpSpeedDecreasing=25, interpSpeedIncreasing=25}
    planarLimits [
     {drivingBone {boneName=dyn_back_strap_jnt_2}, planeTransform {location=<0 0 0>, rotation=<rot 0 0 -90>, scale=<1 1 1>}}
     {drivingBone {boneName=dyn_back_strap_jnt_3}, planeTransform {location=<0 0 0>, rotation=<rot 0 0 -90>, scale=<1 1 1>}}
     {drivingBone {boneName=dyn_back_strap_jnt_4}, planeTransform {location=<0 0 0>, rotation=<rot 0 0 -90>, scale=<1 1 1>}}
     {drivingBone {boneName=dyn_back_strap_jnt_5}, planeTransform {location=<0 0 0>, rotation=<rot 0 0 -90>, scale=<1 1 1>}}
     {drivingBone {boneName=dyn_back_strap_jnt_6}, planeTransform {location=<0 0 0>, rotation=<rot 0 0 -90>, scale=<1 1 1>}}
     {drivingBone {boneName=dyn_back_strap_jnt_7}, planeTransform {location=<0 0 0>, rotation=<rot 0 0 -90>, scale=<1 1 1>}}
    ]
    rotationLimits [{limitMax=<180 180 180>}, {limitMax=<180 180 180>}, {limitMax=<180 180 180>}, {limitMax=<180 180 180>}, {limitMax=<180 180 180>}]
    rotationOffsets [<0 0 0>, <0 0 0>, <0 0 0>, <0 0 0>, <0 0 0>]
    trailBone {boneName=dyn_back_strap_jnt_7}
    trailRelaxationSpeed {editorCurveData {keys [<at 0 15 RCIM_Linear>,<at 0.5289 13.1838 RCIM_Linear>,<at 1 6 RCIM_Linear>]}}
   }
   n5 = Trail controller - Bone: dyn_L_strap_jnt_3!(ComponentPose=@n4, Alpha=1, RelaxationSpeedScale:float=2, FakeVelocity:Vector="0,0,0") -> (Pose) {
    bActorSpaceFakeVel=true
    bShowTrailLocation=true
    bUsePlanarLimit=true
    debugLifeTime=0.1
    maxDeltaTime=0.0333
    alphaScaleBiasClamp {bInterpResult=true, interpSpeedDecreasing=25, interpSpeedIncreasing=25}
    planarLimits [
     {drivingBone {boneName=dyn_L_strap_jnt_3}, planeTransform {location=<0 0 1>, rotation=<rot 0 0 0>, scale=<1 1 1>}}
     {drivingBone {boneName=dyn_L_strap_jnt_2}, planeTransform {location=<0 0 1>, rotation=<rot 0 0 0>, scale=<1 1 1>}}
     {drivingBone {boneName=dyn_L_strap_jnt_1}, planeTransform {location=<0 0 1>, rotation=<rot 0 0 0>, scale=<1 1 1>}}
    ]
    rotationLimits [{limitMax=<180 180 180>}, {limitMax=<180 180 180>}]
    rotationOffsets [<0 0 0>, <0 0 0>]
    trailBone {boneName=dyn_L_strap_jnt_3}
    trailRelaxationSpeed {editorCurveData {keys [<at 0 15 RCIM_Linear>,<at 0.5289 13.1838 RCIM_Linear>,<at 1 6 RCIM_Linear>]}}
   }
   n6 = Trail controller - Bone: dyn_R_strap_jnt_3!(ComponentPose=@n5, Alpha=1, RelaxationSpeedScale:float=2, FakeVelocity:Vector="0,0,0") -> (Pose) {
    bActorSpaceFakeVel=true
    bShowTrailLocation=true
    bUsePlanarLimit=true
    debugLifeTime=0.1
    maxDeltaTime=0.0333
    alphaScaleBiasClamp {bInterpResult=true, interpSpeedDecreasing=25, interpSpeedIncreasing=25}
    planarLimits [
     {drivingBone {boneName=dyn_R_strap_jnt_1}, planeTransform {location=<0 0 -1>, rotation=<rot 0 0 -180>, scale=<1 1 1>}}
     {drivingBone {boneName=dyn_R_strap_jnt_2}, planeTransform {location=<0 0 -1>, rotation=<rot 0 0 180>, scale=<1 1 1>}}
     {drivingBone {boneName=dyn_R_strap_jnt_3}, planeTransform {location=<0 0 -1>, rotation=<rot 0 0 -180>, scale=<1 1 1>}}
    ]
    rotationLimits [{limitMax=<180 180 180>}, {limitMax=<180 180 180>}]
    rotationOffsets [<0 0 0>, <0 0 0>]
    trailBone {boneName=dyn_R_strap_jnt_3}
    trailRelaxationSpeed {editorCurveData {keys [<at 0 15 RCIM_Linear>,<at 0.5289 13.1838 RCIM_Linear>,<at 1 6 RCIM_Linear>]}}
   }
   n7 = Component To Local(ComponentPose=@n6)
   n8 = Get bRigidBodyEnabled()
   n9 = "Transform (Modify) Bone - Bone: dyn_mid_backSkirt_jnt_1"(
    ComponentPose=Local To Component(LocalPose=Blend Poses by bool(BlendPose_0=@n3, BlendPose_1=@n7, BlendTime_0=0.2, BlendTime_1=0.2, bActiveValue=@n8) {blendType=Linear})
    Alpha=1
    Translation="0,0,0"
    Rotation="0,-1.5,0"
    Scale="1,1,1"
   ) {
    rotationMode=BMM_Additive
    rotationSpace=BCS_BoneSpace
    alphaBoolBlend {blendOption=Linear}
    boneToModify {boneName=dyn_mid_backSkirt_jnt_1}
    rotation {yaw=-1.5}
   }
   n10 = "Transform (Modify) Bone - Bone: dyn_loinClothFront_l_MSTR"(ComponentPose=@n9, Alpha=1, Translation="0,-0.55,0", Rotation="10,2,0", Scale="1,1,1") {
    rotationMode=BMM_Additive
    rotationSpace=BCS_BoneSpace
    translationMode=BMM_Additive
    translationSpace=BCS_BoneSpace
    alphaBoolBlend {blendOption=Linear}
    alphaScaleBiasClamp {bInterpResult=true, interpSpeedDecreasing=15, interpSpeedIncreasing=15}
    boneToModify {boneName=dyn_loinClothFront_l_MSTR}
    rotation {pitch=10, yaw=2}
    translation {y=-0.55}
   }
   n11 = "Transform (Modify) Bone - Bone: dyn_loinClothFront_r_MSTR"(ComponentPose=@n10, Alpha=1, Translation="0,0,1", Rotation="-10,1,0", Scale="1,1,1") {
    rotationMode=BMM_Additive
    rotationSpace=BCS_BoneSpace
    translationMode=BMM_Additive
    translationSpace=BCS_BoneSpace
    alphaBoolBlend {blendOption=Linear}
    alphaScaleBiasClamp {bInterpResult=true, interpSpeedDecreasing=15, interpSpeedIncreasing=15}
    boneToModify {boneName=dyn_loinClothFront_r_MSTR}
    rotation {pitch=-10, yaw=1}
    translation {z=1}
   }
   n12 = "Transform (Modify) Bone - Bone: dyn_loinClothFront_c_MSTR"(ComponentPose=@n11, Alpha=1, Translation="0,0,1", Rotation="0,11,0", Scale="1,1,1") {
    rotationMode=BMM_Additive
    rotationSpace=BCS_BoneSpace
    translationMode=BMM_Additive
    translationSpace=BCS_BoneSpace
    alphaBoolBlend {blendOption=Linear}
    alphaScaleBiasClamp {bInterpResult=true, interpSpeedDecreasing=15, interpSpeedIncreasing=15}
    boneToModify {boneName=dyn_loinClothFront_c_MSTR}
    rotation {yaw=11}
    translation {z=1}
   }
   n13 = "Transform (Modify) Bone - Bone: dyn_loinClothBack_l_MSTR"(ComponentPose=@n12, Alpha=1, Translation="0,-0.55,0", Rotation="-10,2,0", Scale="1,1,1") {
    rotationMode=BMM_Additive
    rotationSpace=BCS_BoneSpace
    translationMode=BMM_Additive
    translationSpace=BCS_BoneSpace
    alphaBoolBlend {blendOption=Linear}
    alphaScaleBiasClamp {bInterpResult=true, interpSpeedDecreasing=15, interpSpeedIncreasing=15}
    boneToModify {boneName=dyn_loinClothBack_l_MSTR}
    rotation {pitch=-10, yaw=2}
    translation {y=-0.55}
   }
   n14 = "Transform (Modify) Bone - Bone: dyn_loinClothBack_r_MSTR"(ComponentPose=@n13, Alpha=1, Translation="0,-0.55,0", Rotation="13,2,0", Scale="1,1,1") {
    rotationMode=BMM_Additive
    rotationSpace=BCS_BoneSpace
    translationMode=BMM_Additive
    translationSpace=BCS_BoneSpace
    alphaBoolBlend {blendOption=Linear}
    alphaScaleBiasClamp {bInterpResult=true, interpSpeedDecreasing=15, interpSpeedIncreasing=15}
    boneToModify {boneName=dyn_loinClothBack_r_MSTR}
    rotation {pitch=13, yaw=2}
    translation {y=-0.55}
   }
   Output Pose(Result=Component To Local(ComponentPose=@n14)) {layerGroup=None}
   n15 = Anim Dynamics - Bone: dyn_R_sleeve_jnt_A(ComponentPose=Local To Component(LocalPose=Copy Pose From Mesh() {bUseAttachedParent=true}), Alpha=1, ExternalForce="0,0,0") {
    bGravityOverrideInSimSpace=true
    bOverrideLinearDamping=true
    linearDampingOverride=1
    alphaBoolBlend {blendOption=Linear}
    boundBone {boneName=dyn_R_sleeve_jnt_A}
    componentAppliedLinearAccClamp=<4000 4000 4000>
    componentLinearAccScale=<1 1 1.01>
    componentLinearVelScale=<0.1 0.1 0.1>
    physicsBodyDefinitions [{
      sphereCollisionRadius=0
      boundBone {boneName=dyn_R_sleeve_jnt_A}
      boxExtents=<1 1 1>
      constraintSetup {angularLimitsMax=<0 0 0>, angularLimitsMin=<0 0 0>, angularTarget=<0 0 0>, linearAxesMax=<0.5 0.5 1.5>, linearAxesMin=<-0.5 -0.5 -1.5>}
      localJointOffset=<0 0 0>
     }]
    relativeSpaceBone {boneName=lowerarm_r}
   }
   n16 = Anim Dynamics - Bone: dyn_R_sleeve_jnt_B(ComponentPose=@n15, Alpha=1, ExternalForce="0,0,0") {
    bGravityOverrideInSimSpace=true
    bOverrideLinearDamping=true
    linearDampingOverride=1
    alphaBoolBlend {blendOption=Linear}
    boundBone {boneName=dyn_R_sleeve_jnt_B}
    componentAppliedLinearAccClamp=<4000 4000 4000>
    componentLinearAccScale=<1 1 1.01>
    componentLinearVelScale=<0.1 0.1 0.1>
    physicsBodyDefinitions [{
      sphereCollisionRadius=0
      boundBone {boneName=dyn_R_sleeve_jnt_B}
      boxExtents=<1 1 1>
      constraintSetup {angularLimitsMax=<0 0 0>, angularLimitsMin=<0 0 0>, angularTarget=<0 0 0>, linearAxesMax=<0.5 0.5 1.5>, linearAxesMin=<-0.5 -0.5 -1.5>}
      localJointOffset=<0 0 0>
     }]
    relativeSpaceBone {boneName=lowerarm_r}
   }
   n17 = "Constraint - dyn_loinClothFront_l_MSTR (R)"(ComponentPose=@n16, Alpha=1, ConstraintWeights_0=1) {
    alphaBoolBlend {blendOption=Linear}
    boneToModify {boneName=dyn_loinClothFront_l_MSTR}
    constraintSetup [{transformType=Rotation, targetBone {boneName=thigh_l}}]
    constraintWeights [1]
   }
   n18 = "Constraint - dyn_loinClothFront_r_MSTR (R)"(ComponentPose=@n17, Alpha=1, ConstraintWeights_0=1) {
    alphaBoolBlend {blendOption=Linear}
    boneToModify {boneName=dyn_loinClothFront_r_MSTR}
    constraintSetup [{transformType=Rotation, targetBone {boneName=thigh_r}}]
    constraintWeights [1]
   }
   Save cached pose 'Constrain_Cache'(Pose=Component To Local(ComponentPose=@n18)) {cacheName=Constrain_Cache}
   n19 = RigidBody(ComponentPose=Local To Component(LocalPose=Use cached pose 'Constrain_Cache'()), OverrideWorldGravity="0,0,-980") {
    bClampLinearTranslationLimitToRefPose=true
    bOverrideWorldGravity=true
    overridePhysicsAsset=&/M_MED_Double_Agent_Rear_Physics
    simulationSpace=BaseBoneSpace
    alphaBoolBlend {blendOption=Linear}
    baseBoneRef {boneName=root}
    componentAppliedLinearAccClamp=<1000 1000 1100>
    componentLinearAccScale=<0.35 0.35 0.45>
    componentLinearVelScale=<0.3 0.3 0.45>
    overrideWorldGravity {z=-980}
   }
   n20 = RigidBody(ComponentPose=@n19, OverrideWorldGravity="0,0,-980") {
    bClampLinearTranslationLimitToRefPose=true
    bOverrideWorldGravity=true
    overridePhysicsAsset=&/M_MED_Double_Agent_Rear_FE_Physics
    simulationSpace=BaseBoneSpace
    alphaBoolBlend {blendOption=Linear}
    baseBoneRef {boneName=root}
    componentAppliedLinearAccClamp=<1000 1000 1100>
    componentLinearAccScale=<0.35 0.35 0.45>
    componentLinearVelScale=<0.3 0.3 0.45>
    overrideWorldGravity {z=-980}
   }
   n21 = "Constraint - dyn_loinClothFront_r_MSTR (P)"(ComponentPose=@n20, Alpha=1, ConstraintWeights_0=1) {
    alphaBoolBlend {blendOption=Linear}
    boneToModify {boneName=dyn_loinClothFront_r_MSTR}
    constraintSetup [{targetBone {boneName=thigh_r}}]
    constraintWeights [1]
   }
   n22 = "Constraint - dyn_loinClothFront_c_MSTR (PP)"!(ComponentPose=@n21, Alpha=1, ConstraintWeights_0=0.5, ConstraintWeights_1:float=0.5) -> (Pose) {
    alphaBoolBlend {blendOption=Linear}
    boneToModify {boneName=dyn_loinClothFront_c_MSTR}
    constraintSetup [{targetBone {boneName=thigh_r}}, {targetBone {boneName=thigh_l}}]
    constraintWeights [1,1]
   }
   n23 = "Constraint - dyn_loinClothFront_l_MSTR (P)"(ComponentPose=@n22, Alpha=1, ConstraintWeights_0=1) {
    alphaBoolBlend {blendOption=Linear}
    boneToModify {boneName=dyn_loinClothFront_l_MSTR}
    constraintSetup [{targetBone {boneName=thigh_l}}]
    constraintWeights [1]
   }
   n24 = "Constraint - dyn_loinClothFront_01_r (P)"(ComponentPose=@n23, Alpha=1, ConstraintWeights_0=1) {
    alphaBoolBlend {blendOption=Linear}
    boneToModify {boneName=dyn_loinClothFront_01_r}
    constraintSetup [{targetBone {boneName=thigh_r}}]
    constraintWeights [1]
   }
   n25 = "Constraint - dyn_loinClothFront_01_c (PP)"!(ComponentPose=@n24, Alpha=1, ConstraintWeights_0=0.5, ConstraintWeights_1:float=0.5) -> (Pose) {
    alphaBoolBlend {blendOption=Linear}
    boneToModify {boneName=dyn_loinClothFront_01_c}
    constraintSetup [{targetBone {boneName=thigh_r}}, {targetBone {boneName=thigh_l}}]
    constraintWeights [1,1]
   }
   n26 = "Constraint - dyn_loinClothFront_01_l (P)"(ComponentPose=@n25, Alpha=1, ConstraintWeights_0=1) {
    alphaBoolBlend {blendOption=Linear}
    boneToModify {boneName=dyn_loinClothFront_01_l}
    constraintSetup [{targetBone {boneName=thigh_l}}]
    constraintWeights [1]
   }
   n27 = "Constraint - dyn_loinClothFront_02_r (P)"(ComponentPose=@n26, Alpha=1, ConstraintWeights_0=1) {
    alphaBoolBlend {blendOption=Linear}
    boneToModify {boneName=dyn_loinClothFront_02_r}
    constraintSetup [{targetBone {boneName=thigh_r}}]
    constraintWeights [1]
   }
   n28 = "Constraint - dyn_loinClothFront_02_c (PP)"!(ComponentPose=@n27, Alpha=1, ConstraintWeights_0=0.5, ConstraintWeights_1:float=0.5) -> (Pose) {
    alphaBoolBlend {blendOption=Linear}
    boneToModify {boneName=dyn_loinClothFront_02_c}
    constraintSetup [{targetBone {boneName=thigh_r}}, {targetBone {boneName=thigh_l}}]
    constraintWeights [1,1]
   }
   n29 = "Constraint - dyn_loinClothFront_02_l (P)"(ComponentPose=@n28, Alpha=1, ConstraintWeights_0=1) {
    alphaBoolBlend {blendOption=Linear}
    boneToModify {boneName=dyn_loinClothFront_02_l}
    constraintSetup [{targetBone {boneName=thigh_l}}]
    constraintWeights [1]
   }
   n30 = "Constraint - dyn_loinClothFront_03_r (P)"(ComponentPose=@n29, Alpha=1, ConstraintWeights_0=1) {
    alphaBoolBlend {blendOption=Linear}
    boneToModify {boneName=dyn_loinClothFront_03_r}
    constraintSetup [{targetBone {boneName=thigh_r}}]
    constraintWeights [1]
   }
   n31 = "Constraint - dyn_loinClothFront_03_c (PP)"!(ComponentPose=@n30, Alpha=1, ConstraintWeights_0=0.5, ConstraintWeights_1:float=0.5) -> (Pose) {
    alphaBoolBlend {blendOption=Linear}
    boneToModify {boneName=dyn_loinClothFront_03_c}
    constraintSetup [{targetBone {boneName=thigh_r}}, {targetBone {boneName=thigh_l}}]
    constraintWeights [1,1]
   }
   n32 = "Constraint - dyn_loinClothFront_03_l (P)"(ComponentPose=@n31, Alpha=1, ConstraintWeights_0=1) {
    alphaBoolBlend {blendOption=Linear}
    boneToModify {boneName=dyn_loinClothFront_03_l}
    constraintSetup [{targetBone {boneName=thigh_l}}]
    constraintWeights [1]
   }
   n33 = "Transform (Modify) Bone - Bone: dyn_loinClothFront_01_l"(ComponentPose=@n32, Alpha=1, Translation="-1.014,-0.013,1.705", Rotation="6.569,-0.179,1.930", Scale="1,1,1") {
    rotationMode=BMM_Additive
    rotationSpace=BCS_BoneSpace
    scaleMode=BMM_Additive
    translationMode=BMM_Additive
    translationSpace=BCS_BoneSpace
    alphaBoolBlend {blendOption=Linear}
    boneToModify {boneName=dyn_loinClothFront_01_l}
    rotation=<rot 6.569 -0.179 1.93>
    translation=<-1.014 -0.013 1.705>
   }
   n34 = "Transform (Modify) Bone - Bone: dyn_loinClothFront_01_r"(ComponentPose=@n33, Alpha=1, Translation="0.718,1.010,-7.934", Rotation="19.442,-9.050,-24.492", Scale="1,1,1") {
    rotationMode=BMM_Additive
    rotationSpace=BCS_BoneSpace
    scaleMode=BMM_Additive
    translationMode=BMM_Additive
    translationSpace=BCS_BoneSpace
    alphaBoolBlend {blendOption=Linear}
    boneToModify {boneName=dyn_loinClothFront_01_r}
    rotation=<rot 19.442 -9.05 -24.492>
    translation=<0.718 1.01 -7.934>
   }
   n35 = "Transform (Modify) Bone - Bone: dyn_loinClothFront_01_l"(ComponentPose=@n34, Alpha=1, Translation="-1.014,-0.013,1.705", Rotation="2.185,-30,4.758", Scale="1,1,1") {
    rotationMode=BMM_Additive
    rotationSpace=BCS_BoneSpace
    scaleMode=BMM_Additive
    translationMode=BMM_Additive
    translationSpace=BCS_BoneSpace
    alphaBoolBlend {blendOption=Linear}
    boneToModify {boneName=dyn_loinClothFront_01_l}
    rotation=<rot 2.185 -30 4.758>
    translation=<-1.014 -0.013 1.705>
   }
   n36 = "Transform (Modify) Bone - Bone: dyn_loinClothFront_01_r"(ComponentPose=@n35, Alpha=1, Translation="-3.442,-1.941,-0.493", Rotation="18.517,-36.622,5.178", Scale="1,1,1") {
    rotationMode=BMM_Additive
    rotationSpace=BCS_BoneSpace
    scaleMode=BMM_Additive
    translationMode=BMM_Additive
    translationSpace=BCS_BoneSpace
    alphaBoolBlend {blendOption=Linear}
    boneToModify {boneName=dyn_loinClothFront_01_r}
    rotation=<rot 18.517 -36.622 5.178>
    translation=<-3.442 -1.941 -0.493>
   }
   n37 = "Constraint - dyn_loinClothFront_01_c (PP)"(ComponentPose=@n36, Alpha=1, ConstraintWeights_0=0.5, ConstraintWeights_1=0.5) {
    alphaBoolBlend {blendOption=Linear}
    boneToModify {boneName=dyn_loinClothFront_01_c}
    constraintSetup [{targetBone {boneName=dyn_loinClothFront_01_l}}, {targetBone {boneName=dyn_loinClothFront_01_r}}]
    constraintWeights [1,1]
   }
   n38 = "Transform (Modify) Bone - Bone: dyn_loinClothFront_c_MSTR"(ComponentPose=@n37, Alpha=1, Translation="0.000,-1,0", Rotation="0,0,0", Scale="1,1,1") {
    rotationMode=BMM_Additive
    rotationSpace=BCS_BoneSpace
    scaleMode=BMM_Additive
    translationMode=BMM_Additive
    translationSpace=BCS_BoneSpace
    alphaBoolBlend {blendOption=Linear}
    boneToModify {boneName=dyn_loinClothFront_c_MSTR}
    translation {y=-1}
   }
   n39 = "Transform (Modify) Bone - Bone: dyn_loinClothFront_c_MSTR"(ComponentPose=@n38, Alpha=1, Translation="0.000,-3,0", Rotation="0,-5,0", Scale="1,1,1") {
    rotationMode=BMM_Additive
    rotationSpace=BCS_BoneSpace
    scaleMode=BMM_Additive
    translationMode=BMM_Additive
    translationSpace=BCS_BoneSpace
    alphaBoolBlend {blendOption=Linear}
    alphaScaleBiasClamp {bInterpResult=true, interpSpeedDecreasing=15, interpSpeedIncreasing=15}
    boneToModify {boneName=dyn_loinClothFront_c_MSTR}
    rotation {yaw=-5}
    translation {y=-3}
   }
   n40 = Component To Local(ComponentPose=@n39)
   n41 = "Constraint - dyn_loinClothFront_l_MSTR (P)"(ComponentPose=Local To Component(LocalPose=Use cached pose 'Constrain_Cache'()), Alpha=1, ConstraintWeights_0=1) {
    alphaBoolBlend {blendOption=Linear}
    boneToModify {boneName=dyn_loinClothFront_l_MSTR}
    constraintSetup [{targetBone {boneName=thigh_twist_01_l}}]
    constraintWeights [1]
   }
   n42 = "Constraint - dyn_loinClothFront_01_l (P)"(ComponentPose=@n41, Alpha=1, ConstraintWeights_0=1) {
    alphaBoolBlend {blendOption=Linear}
    boneToModify {boneName=dyn_loinClothFront_01_l}
    constraintSetup [{targetBone {boneName=thigh_twist_01_l}}]
    constraintWeights [1]
   }
   n43 = "Constraint - dyn_loinClothFront_02_l (P)"(ComponentPose=@n42, Alpha=1, ConstraintWeights_0=1) {
    alphaBoolBlend {blendOption=Linear}
    boneToModify {boneName=dyn_loinClothFront_02_l}
    constraintSetup [{targetBone {boneName=thigh_twist_01_l}}]
    constraintWeights [1]
   }
   n44 = "Constraint - dyn_loinClothFront_03_l (P)"(ComponentPose=@n43, Alpha=1, ConstraintWeights_0=1) {
    alphaBoolBlend {blendOption=Linear}
    boneToModify {boneName=dyn_loinClothFront_03_l}
    constraintSetup [{targetBone {boneName=thigh_twist_01_l}}]
    constraintWeights [1]
   }
   n45 = "Constraint - dyn_loinClothFront_r_MSTR (PP)"!(ComponentPose=@n44, Alpha=1, ConstraintWeights_0=0.5, ConstraintWeights_1:float=0.5) -> (Pose) {
    alphaBoolBlend {blendOption=Linear}
    boneToModify {boneName=dyn_loinClothFront_r_MSTR}
    constraintSetup [{targetBone {boneName=thigh_twist_01_r}}, {targetBone {boneName=dyn_loinClothFront_r_MSTR}}]
    constraintWeights [1,1]
   }
   n46 = "Constraint - dyn_loinClothFront_01_r (PP)"!(ComponentPose=@n45, Alpha=1, ConstraintWeights_0=0.75, ConstraintWeights_1:float=0.25) -> (Pose) {
    alphaBoolBlend {blendOption=Linear}
    boneToModify {boneName=dyn_loinClothFront_01_r}
    constraintSetup [{targetBone {boneName=thigh_twist_01_r}}, {targetBone {boneName=dyn_loinClothFront_01_r}}]
    constraintWeights [1,1]
   }
   n47 = "Constraint - dyn_loinClothFront_02_r (PP)"!(ComponentPose=@n46, Alpha=1, ConstraintWeights_0=0.75, ConstraintWeights_1:float=0.25) -> (Pose) {
    alphaBoolBlend {blendOption=Linear}
    boneToModify {boneName=dyn_loinClothFront_02_r}
    constraintSetup [{targetBone {boneName=thigh_twist_01_r}}, {targetBone {boneName=dyn_loinClothFront_02_r}}]
    constraintWeights [1,1]
   }
   n48 = "Constraint - dyn_loinClothFront_03_r (PP)"!(ComponentPose=@n47, Alpha=1, ConstraintWeights_0=0.5, ConstraintWeights_1:float=0.5) -> (Pose) {
    alphaBoolBlend {blendOption=Linear}
    boneToModify {boneName=dyn_loinClothFront_03_r}
    constraintSetup [{targetBone {boneName=thigh_twist_01_r}}, {targetBone {boneName=dyn_loinClothFront_03_r}}]
    constraintWeights [1,1]
   }
   n49 = "Constraint - dyn_loinClothFront_c_MSTR (PP)"(ComponentPose=@n48, Alpha=1, ConstraintWeights_0=0.5, ConstraintWeights_1=0.5) {
    alphaBoolBlend {blendOption=Linear}
    boneToModify {boneName=dyn_loinClothFront_c_MSTR}
    constraintSetup [{targetBone {boneName=thigh_twist_01_l}}, {targetBone {boneName=dyn_loinClothFront_r_MSTR}}]
    constraintWeights [1,1]
   }
   n50 = "Constraint - dyn_loinClothFront_01_c (PP)"(ComponentPose=@n49, Alpha=1, ConstraintWeights_0=0.5, ConstraintWeights_1=0.5) {
    alphaBoolBlend {blendOption=Linear}
    boneToModify {boneName=dyn_loinClothFront_01_c}
    constraintSetup [{targetBone {boneName=thigh_twist_01_l}}, {targetBone {boneName=dyn_loinClothFront_01_r}}]
    constraintWeights [1,1]
   }
   n51 = "Constraint - dyn_loinClothFront_02_c (PP)"(ComponentPose=@n50, Alpha=1, ConstraintWeights_0=0.5, ConstraintWeights_1=0.5) {
    alphaBoolBlend {blendOption=Linear}
    boneToModify {boneName=dyn_loinClothFront_02_c}
    constraintSetup [{targetBone {boneName=thigh_twist_01_l}}, {targetBone {boneName=dyn_loinClothFront_02_r}}]
    constraintWeights [1,1]
   }
   n52 = "Constraint - dyn_loinClothFront_03_c (PP)"(ComponentPose=@n51, Alpha=1, ConstraintWeights_0=0.5, ConstraintWeights_1=0.5) {
    alphaBoolBlend {blendOption=Linear}
    boneToModify {boneName=dyn_loinClothFront_03_c}
    constraintSetup [{targetBone {boneName=thigh_twist_01_l}}, {targetBone {boneName=dyn_loinClothFront_03_r}}]
    constraintWeights [1,1]
   }
   n53 = RigidBody!(ComponentPose=@n52, Alpha=0.5, ExternalForce=Get Gravity_Override() -> (Gravity_Override:Vector)) -> (Pose) {
    bClampLinearTranslationLimitToRefPose=true
    bOverrideWorldGravity=true
    overridePhysicsAsset=&M_MED_Double_Agent_Rear_Physics
    simulationSpace=BaseBoneSpace
    alphaBoolBlend {blendOption=Linear}
    baseBoneRef {boneName=root}
    componentAppliedLinearAccClamp=<2000 2000 3000>
    overrideWorldGravity {z=-980}
   }
   n54 = "Transform (Modify) Bone - Bone: dyn_loinClothFront_l_MSTR"(ComponentPose=@n53, Alpha=1, Translation="0,6,2", Rotation="0,15,0", Scale="1,1,1") {
    rotationMode=BMM_Additive
    scaleMode=BMM_Additive
    translationMode=BMM_Additive
    alphaBoolBlend {blendOption=Linear}
    boneToModify {boneName=dyn_loinClothFront_l_MSTR}
    rotation {yaw=15}
    translation {y=6, z=2}
   }
   n55 = "Transform (Modify) Bone - Bone: dyn_loinClothFront_c_MSTR"(ComponentPose=@n54, Alpha=1, Translation="0,4,4", Rotation="0,0,0", Scale="1,1,1") {
    rotationMode=BMM_Additive
    scaleMode=BMM_Additive
    translationMode=BMM_Additive
    alphaBoolBlend {blendOption=Linear}
    boneToModify {boneName=dyn_loinClothFront_c_MSTR}
    translation {y=4, z=4}
   }
   n56 = "Transform (Modify) Bone - Bone: dyn_loinClothFront_r_MSTR"(ComponentPose=@n55, Alpha=1, Translation="0,1,2", Rotation="0,0,0", Scale="1,1,1") {
    rotationMode=BMM_Additive
    scaleMode=BMM_Additive
    translationMode=BMM_Additive
    alphaBoolBlend {blendOption=Linear}
    boneToModify {boneName=dyn_loinClothFront_r_MSTR}
    translation {y=1, z=2}
   }
   n57 = "Transform (Modify) Bone - Bone: dyn_loinClothFront_01_l"(ComponentPose=@n56, Alpha=1, Translation="0,0,1", Rotation="0,0,0", Scale="1,1,1") {
    rotationMode=BMM_Additive
    scaleMode=BMM_Additive
    translationMode=BMM_Additive
    alphaBoolBlend {blendOption=Linear}
    boneToModify {boneName=dyn_loinClothFront_01_l}
    translation {z=1}
   }
   n58 = "Transform (Modify) Bone - Bone: dyn_loinClothFront_01_c"(ComponentPose=@n57, Alpha=1, Translation="0,0,0", Rotation="0,0,0", Scale="1,1,1") {
    rotationMode=BMM_Additive
    scaleMode=BMM_Additive
    translationMode=BMM_Additive
    alphaBoolBlend {blendOption=Linear}
    boneToModify {boneName=dyn_loinClothFront_01_c}
   }
   n59 = "Transform (Modify) Bone - Bone: dyn_loinClothFront_01_r"(ComponentPose=@n58, Alpha=1, Translation="0,0,0", Rotation="0,0,0", Scale="1,1,1") {
    rotationMode=BMM_Additive
    scaleMode=BMM_Additive
    translationMode=BMM_Additive
    alphaBoolBlend {blendOption=Linear}
    boneToModify {boneName=dyn_loinClothFront_01_r}
   }
   n60 = "Transform (Modify) Bone - Bone: dyn_loinClothFront_02_l"(ComponentPose=@n59, Alpha=1, Translation="0,0,1", Rotation="0,0,10", Scale="1,1,1") {
    rotationMode=BMM_Additive
    scaleMode=BMM_Additive
    translationMode=BMM_Additive
    alphaBoolBlend {blendOption=Linear}
    boneToModify {boneName=dyn_loinClothFront_02_l}
    rotation {roll=10}
    translation {z=1}
   }
   n61 = "Transform (Modify) Bone - Bone: dyn_loinClothFront_02_c"(ComponentPose=@n60, Alpha=1, Translation="0,0,0", Rotation="0,0,0", Scale="1,1,1") {
    rotationMode=BMM_Additive
    scaleMode=BMM_Additive
    translationMode=BMM_Additive
    alphaBoolBlend {blendOption=Linear}
    boneToModify {boneName=dyn_loinClothFront_02_c}
   }
   n62 = "Transform (Modify) Bone - Bone: dyn_loinClothFront_02_r"(ComponentPose=@n61, Alpha=1, Translation="0,0,-3", Rotation="0,0,0", Scale="1,1,1") {
    rotationMode=BMM_Additive
    scaleMode=BMM_Additive
    translationMode=BMM_Additive
    alphaBoolBlend {blendOption=Linear}
    boneToModify {boneName=dyn_loinClothFront_02_r}
    translation {z=-3}
   }
   n63 = "Transform (Modify) Bone - Bone: dyn_loinClothFront_03_l"(ComponentPose=@n62, Alpha=1, Translation="0,0,0", Rotation="0,0,0", Scale="1,1,1") {
    rotationMode=BMM_Additive
    scaleMode=BMM_Additive
    translationMode=BMM_Additive
    alphaBoolBlend {blendOption=Linear}
    boneToModify {boneName=dyn_loinClothFront_03_l}
   }
   n64 = "Transform (Modify) Bone - Bone: dyn_loinClothFront_03_c"(ComponentPose=@n63, Alpha=1, Translation="0,0,0", Rotation="0,0,0", Scale="1,1,1") {
    rotationMode=BMM_Additive
    scaleMode=BMM_Additive
    translationMode=BMM_Additive
    alphaBoolBlend {blendOption=Linear}
    boneToModify {boneName=dyn_loinClothFront_03_c}
   }
   n65 = "Transform (Modify) Bone - Bone: dyn_loinClothFront_03_r"(ComponentPose=@n64, Alpha=1, Translation="0,0,0", Rotation="0,0,0", Scale="1,1,1") {
    rotationMode=BMM_Additive
    scaleMode=BMM_Additive
    translationMode=BMM_Additive
    alphaBoolBlend {blendOption=Linear}
    boneToModify {boneName=dyn_loinClothFront_03_r}
   }
   n66 = "Transform (Modify) Bone - Bone: dyn_R_strap_jnt_1"(ComponentPose=@n65, Alpha=1, Translation="0,0,0", Rotation="0,-23,0", Scale="1,1,1") {
    rotationMode=BMM_Additive
    rotationSpace=BCS_BoneSpace
    alphaBoolBlend {blendOption=Linear}
    alphaScaleBiasClamp {bInterpResult=true, interpSpeedDecreasing=15, interpSpeedIncreasing=8}
    boneToModify {boneName=dyn_R_strap_jnt_1}
    rotation {yaw=-23}
   }
   n67 = "Transform (Modify) Bone - Bone: dyn_L_strap_jnt_1"(ComponentPose=@n66, Alpha=1, Translation="0,-6,2", Rotation="5,-95,0", Scale="1,1,1") {
    rotationMode=BMM_Additive
    rotationSpace=BCS_BoneSpace
    translationMode=BMM_Additive
    translationSpace=BCS_BoneSpace
    alphaBoolBlend {blendOption=Linear}
    alphaScaleBiasClamp {bInterpResult=true, interpSpeedDecreasing=6, interpSpeedIncreasing=15}
    boneToModify {boneName=dyn_L_strap_jnt_1}
    rotation {pitch=5, yaw=-95}
    translation {y=-6, z=2}
   }
   n68 = "Transform (Modify) Bone - Bone: dyn_loinClothFront_l_MSTR"(ComponentPose=@n67, Alpha=1, Translation="-0.5,0,1", Rotation="0,2,0", Scale="1,1,1") {
    rotationMode=BMM_Additive
    rotationSpace=BCS_BoneSpace
    scaleMode=BMM_Additive
    translationMode=BMM_Additive
    translationSpace=BCS_BoneSpace
    alphaBoolBlend {blendOption=Linear}
    alphaScaleBiasClamp {bInterpResult=true, interpSpeedDecreasing=15, interpSpeedIncreasing=15}
    boneToModify {boneName=dyn_loinClothFront_l_MSTR}
    rotation {yaw=2}
    translation {x=-0.5, z=1}
   }
   n69 = "Transform (Modify) Bone - Bone: dyn_loinClothFront_01_l"(ComponentPose=@n68, Alpha=1, Translation="-2.5,-0.25,1", Rotation="1,0,15", Scale="1,1,1") {
    rotationMode=BMM_Additive
    rotationSpace=BCS_BoneSpace
    scaleMode=BMM_Additive
    translationMode=BMM_Additive
    translationSpace=BCS_BoneSpace
    alphaBoolBlend {blendOption=Linear}
    alphaScaleBiasClamp {bInterpResult=true, interpSpeedDecreasing=15, interpSpeedIncreasing=15}
    boneToModify {boneName=dyn_loinClothFront_01_l}
    rotation {pitch=1, roll=15}
    translation=<-2.5 -0.25 1>
   }
   n70 = "Transform (Modify) Bone - Bone: dyn_loinClothFront_l_MSTR"(ComponentPose=@n69, Alpha=1, Translation="3,0,5", Rotation="-15,10,0", Scale="1,1,1") {
    rotationMode=BMM_Additive
    rotationSpace=BCS_BoneSpace
    scaleMode=BMM_Additive
    translationMode=BMM_Additive
    translationSpace=BCS_BoneSpace
    alphaBoolBlend {blendOption=Linear}
    alphaScaleBiasClamp {bInterpResult=true, interpSpeedDecreasing=15, interpSpeedIncreasing=15}
    boneToModify {boneName=dyn_loinClothFront_l_MSTR}
    rotation {pitch=-15, yaw=10}
    translation {x=3, z=5}
   }
   n71 = "Transform (Modify) Bone - Bone: dyn_loinClothFront_01_r"(ComponentPose=@n70, Alpha=1, Translation="0.718,1.010,-7.934", Rotation="19.5,-9.05,-24.5", Scale="1,1,1") {
    rotationMode=BMM_Additive
    rotationSpace=BCS_BoneSpace
    scaleMode=BMM_Additive
    translationMode=BMM_Additive
    translationSpace=BCS_BoneSpace
    alphaBoolBlend {blendOption=Linear}
    alphaScaleBiasClamp {bInterpResult=true, interpSpeedDecreasing=15, interpSpeedIncreasing=15}
    boneToModify {boneName=dyn_loinClothFront_01_r}
    rotation=<rot 19.5 -9.05 -24.5>
    translation=<0.718 1.01 -7.934>
   }
   n72 = "Constraint - dyn_loinClothFront_01_c (PP)"(ComponentPose=@n71, Alpha=0, ConstraintWeights_0=0.5, ConstraintWeights_1=0.5) {
    alphaBoolBlend {blendOption=Linear}
    boneToModify {boneName=dyn_loinClothFront_01_c}
    constraintSetup [{targetBone {boneName=dyn_loinClothFront_01_l}}, {targetBone {boneName=dyn_loinClothFront_01_r}}]
    constraintWeights [1,1]
   }
   n73 = Component To Local(ComponentPose=@n72)
   n74 = Get bIsCrouching()
   n75 = Anim Dynamics - Bone: dyn_loinClothFront_02_l(
    ComponentPose=Local To Component(LocalPose=Blend Poses by bool(BlendPose_0=@n73, BlendPose_1=@n40, BlendTime_1=0.25, bActiveValue=@n74) {blendType=Linear})
    Alpha=1
    ExternalForce="0,0,0"
   ) {
    angularDampingOverride=0.9999
    angularSpringConstant=350
    bAngularSpring=true
    bGravityOverrideInSimSpace=true
    bOverrideAngularDamping=true
    simulationSpace=RootRelative
    alphaBoolBlend {blendOption=Linear}
    boundBone {boneName=dyn_loinClothFront_02_l}
    physicsBodyDefinitions [{
      sphereCollisionRadius=0
      boundBone {boneName=dyn_loinClothFront_02_l}
      boxExtents=<8 15 15>
      constraintSetup {angularLimitsMax=<5 5 0>, angularLimitsMin=<-5 -5 -5>, angularTarget=<0 0 0>}
      localJointOffset=<2 0 0>
     }]
   }
   n76 = Anim Dynamics - Bone: dyn_loinClothFront_02_c(ComponentPose=@n75, Alpha=1, ExternalForce="0,0,0") {
    angularDampingOverride=0.9999
    angularSpringConstant=350
    bAngularSpring=true
    bGravityOverrideInSimSpace=true
    bOverrideAngularDamping=true
    simulationSpace=RootRelative
    alphaBoolBlend {blendOption=Linear}
    boundBone {boneName=dyn_loinClothFront_02_c}
    physicsBodyDefinitions [{
      sphereCollisionRadius=0
      boundBone {boneName=dyn_loinClothFront_02_c}
      boxExtents=<8 15 15>
      constraintSetup {angularLimitsMax=<5 10 0>, angularLimitsMin=<-5 -10 -45>, angularTarget=<0 0 0>}
      localJointOffset=<2 0 0>
     }]
   }
   n77 = Anim Dynamics - Bone: dyn_loinClothFront_02_r(ComponentPose=@n76, Alpha=1, ExternalForce="0,0,0") {
    angularDampingOverride=0.9999
    angularSpringConstant=350
    bAngularSpring=true
    bGravityOverrideInSimSpace=true
    bOverrideAngularDamping=true
    simulationSpace=RootRelative
    alphaBoolBlend {blendOption=Linear}
    boundBone {boneName=dyn_loinClothFront_02_r}
    physicsBodyDefinitions [{
      sphereCollisionRadius=0
      boundBone {boneName=dyn_loinClothFront_02_r}
      boxExtents=<8 15 15>
      constraintSetup {angularLimitsMax=<5 20 0>, angularLimitsMin=<-5 -20 -45>, angularTarget=<0 0 0>}
      localJointOffset=<2 0 0>
     }]
   }
   n78 = Anim Dynamics - Bone: dyn_loinClothFront_03_l(ComponentPose=@n77, Alpha=1, ExternalForce="0,0,0") {
    angularDampingOverride=0.9999
    angularSpringConstant=350
    bAngularSpring=true
    bGravityOverrideInSimSpace=true
    bOverrideAngularDamping=true
    simulationSpace=RootRelative
    alphaBoolBlend {blendOption=Linear}
    boundBone {boneName=dyn_loinClothFront_03_l}
    physicsBodyDefinitions [{
      sphereCollisionRadius=0
      boundBone {boneName=dyn_loinClothFront_03_l}
      boxExtents=<8 15 15>
      constraintSetup {angularLimitsMax=<5 5 0>, angularLimitsMin=<-5 -5 -5>, angularTarget=<0 0 0>}
      localJointOffset=<2 0 0>
     }]
   }
   n79 = Anim Dynamics - Bone: dyn_loinClothFront_03_c(ComponentPose=@n78, Alpha=1, ExternalForce="0,0,0") {
    angularDampingOverride=0.9999
    angularSpringConstant=350
    bAngularSpring=true
    bGravityOverrideInSimSpace=true
    bOverrideAngularDamping=true
    simulationSpace=RootRelative
    alphaBoolBlend {blendOption=Linear}
    boundBone {boneName=dyn_loinClothFront_03_c}
    physicsBodyDefinitions [{
      sphereCollisionRadius=0
      boundBone {boneName=dyn_loinClothFront_03_c}
      boxExtents=<8 15 15>
      constraintSetup {angularLimitsMax=<5 10 0>, angularLimitsMin=<-5 -10 -45>, angularTarget=<0 0 0>}
      localJointOffset=<2 0 0>
     }]
   }
   n80 = Anim Dynamics - Bone: dyn_loinClothFront_03_r(ComponentPose=@n79, Alpha=1, ExternalForce="0,0,0") {
    angularDampingOverride=0.9999
    angularSpringConstant=350
    bAngularSpring=true
    bGravityOverrideInSimSpace=true
    bOverrideAngularDamping=true
    simulationSpace=RootRelative
    alphaBoolBlend {blendOption=Linear}
    boundBone {boneName=dyn_loinClothFront_03_r}
    physicsBodyDefinitions [{
      sphereCollisionRadius=0
      boundBone {boneName=dyn_loinClothFront_03_r}
      boxExtents=<8 15 15>
      constraintSetup {angularLimitsMax=<5 20 0>, angularLimitsMin=<-5 -20 -45>, angularTarget=<0 0 0>}
      localJointOffset=<2 0 0>
     }]
   }
   n81 = "Constraint - dyn_mid_backSkirt_jnt_1 (P)"(ComponentPose=@n80, Alpha=0.6, ConstraintWeights_0=1) {
    alphaBoolBlend {blendOption=Linear}
    boneToModify {boneName=dyn_mid_backSkirt_jnt_1}
    constraintSetup [{targetBone {boneName=dyn_loinClothBack_01_c}}]
    constraintWeights [1]
   }
   n82 = "Constraint - dyn_back_strap_jnt_2 (P)"(ComponentPose=@n81, Alpha=0.6, ConstraintWeights_0=1) {
    alphaBoolBlend {blendOption=Linear}
    boneToModify {boneName=dyn_back_strap_jnt_2}
    constraintSetup [{targetBone {boneName=dyn_loinClothBack_02_c}}]
    constraintWeights [1]
   }
   n83 = "Constraint - dyn_back_strap_jnt_3 (P)"(ComponentPose=@n82, Alpha=0.6, ConstraintWeights_0=1) {
    alphaBoolBlend {blendOption=Linear}
    boneToModify {boneName=dyn_back_strap_jnt_3}
    constraintSetup [{targetBone {boneName=dyn_loinClothBack_02_c}}]
    constraintWeights [1]
   }
   Save cached pose 'RBANenabled'(Pose=Component To Local(ComponentPose=@n83)) {cacheName=RBANenabled}
  ]}
 EventGraph {nodes [
   Event BlueprintUpdateAnimation()
   > Get!(execute:Exec) -> (then:Exec, else:Exec, Skel_Mesh:Skeletal Mesh Component Object Reference)
   > n1 = Sequence!(execute) -> (then_0, then_1, then_2:Exec, then_3:Exec, then_4:Exec)
   > n2 = Branch(Condition=Get bIsInFrontEnd())
   > Set fe_alpha(fe_alpha:double=1) -> (Output_Get:double)
   > Set fe_alpha_inv(fe_alpha_inv:double=0) -> (Output_Get:double)
   > Branch(Condition=Get bIsPlayingEmote())
   > Set fe_alpha(fe_alpha=0)
   Set fe_alpha(execute=@n2.else, fe_alpha=0)
   > Set fe_alpha_inv(fe_alpha_inv=1)
   n3 = Branch(execute=@n1.then_1, Condition=Get bIsInFrontEnd())
   > n4 = Gravity Override State Function DEPRECATED(
    CharacterPartAnimInstance=Self-Reference()
    GameTimeInSeconds=GetGameTimeInSeconds()
    GravityOverrideParams=Get Gravity Override Params_0() -> (Gravity Override Params_0:Gravity Override Params Struct Structure)
   )
   > Set Gravity_Override_FE(Gravity_Override_FE:Vector=@n4) -> (Output_Get:Vector)
   > Set Gravity_Override(Gravity_Override:Vector="0, 0,-980") -> (Output_Get:Vector)
   n5 = Gravity Override State Function DEPRECATED!(
    execute=@n3.else
    CharacterPartAnimInstance=Self-Reference()
    GameTimeInSeconds=GetGameTimeInSeconds()
    GravityOverrideParams_Skydive_GoingUp_28_#1_WindFrequency_28_#2:Vector="1.5,7,8"
    GravityOverrideParams_Skydive_GoingUp_28_#1_WindAmplitude_31_#3:Vector="200,-500,400"
    GravityOverrideParams_Skydive_GoingUp_28_#1_WindOffset_32_#4:Vector="0,0,0"
    GravityOverrideParams_Skydive_GoingUp_28_#1_GravityOverride_15_#5:Vector="0,0,-100"
    GravityOverrideParams_Skydive_GoingUp_28_#1_JointName_27_#6:Name=root
    GravityOverrideParams_Skydive_FloatingDown_29_#7_WindFrequency_28_#2:Vector="1.5,7,8"
    GravityOverrideParams_Skydive_FloatingDown_29_#7_WindAmplitude_31_#3:Vector="200,-500,400"
    GravityOverrideParams_Skydive_FloatingDown_29_#7_WindOffset_32_#4:Vector="0,-600,400"
    GravityOverrideParams_Skydive_FloatingDown_29_#7_GravityOverride_15_#5:Vector="0,0,980"
    GravityOverrideParams_Skydive_FloatingDown_29_#7_JointName_27_#6:Name=root
    GravityOverrideParams_Skydive_DivingDown_30_#8_WindFrequency_28_#2:Vector="1.5,7,8"
    GravityOverrideParams_Skydive_DivingDown_30_#8_WindAmplitude_31_#3:Vector="50,300,100"
    GravityOverrideParams_Skydive_DivingDown_30_#8_WindOffset_32_#4:Vector="0,0,0"
    GravityOverrideParams_Skydive_DivingDown_30_#8_GravityOverride_15_#5:Vector="0,0,980"
    GravityOverrideParams_Skydive_DivingDown_30_#8_JointName_27_#6:Name=pelvis
    GravityOverrideParams_Skydive_Parachuting_31_#9_WindFrequency_28_#2:Vector="1.5,7,8"
    GravityOverrideParams_Skydive_Parachuting_31_#9_WindAmplitude_31_#3:Vector="200,-500,400"
    GravityOverrideParams_Skydive_Parachuting_31_#9_WindOffset_32_#4:Vector="0,0,0"
    GravityOverrideParams_Skydive_Parachuting_31_#9_GravityOverride_15_#5:Vector="0,-800,-980"
    GravityOverrideParams_Skydive_Parachuting_31_#9_JointName_27_#6:Name=root
    GravityOverrideParams_OnGround_Standing_32_#10:Gravity Override Setting Struct Structure=(WindFrequency_28_#2=(X=0,Y=0,Z=0),WindAmplitude_31_#3=(X=0,Y=0,Z=0),WindOffset_32_#4=(X=0,Y=0,Z=0),GravityOverride_15_#5=(X=0,Y=0,Z=-980),JointName_27_#6="root")
    GravityOverrideParams_OnGround_Moving_33_#11:Gravity Override Setting Struct Structure=(WindFrequency_28_#2=(X=0,Y=0,Z=0),WindAmplitude_31_#3=(X=0,Y=0,Z=0),WindOffset_32_#4=(X=0,Y=0,Z=0),GravityOverride_15_#5=(X=0,Y=0,Z=-980),JointName_27_#6="root")
    GravityOverrideParams_EmoteOrMelee_34_#12:Gravity Override Setting Struct Structure=(WindFrequency_28_#2=(X=0,Y=0,Z=0),WindAmplitude_31_#3=(X=0,Y=0,Z=0),WindOffset_32_#4=(X=0,Y=0,Z=0),GravityOverride_15_#5=(X=0,Y=0,Z=-980),JointName_27_#6="root")
    GravityOverrideParams_Swimming_35_#13:Gravity Override Setting Struct Structure=(WindFrequency_28_#2=(X=0,Y=0,Z=0),WindAmplitude_31_#3=(X=0,Y=0,Z=0),WindOffset_32_#4=(X=0,Y=0,Z=0),GravityOverride_15_#5=(X=-800,Y=0,Z=0),JointName_27_#6="pelvis")
    GravityOverrideParams_Front_End_36_#14:Gravity Override Setting Struct Structure=(WindFrequency_28_#2=(X=0,Y=0,Z=0),WindAmplitude_31_#3=(X=0,Y=0,Z=0),WindOffset_32_#4=(X=0,Y=0,Z=0),GravityOverride_15_#5=(X=0,Y=0,Z=-980),JointName_27_#6="root")
    GravityOverrideParams_DBNO_40_#15:Gravity Override Setting Struct Structure=(WindFrequency_28_#2=(X=0,Y=0,Z=0),WindAmplitude_31_#3=(X=0,Y=0,Z=0),WindOffset_32_#4=(X=0,Y=0,Z=0),GravityOverride_15_#5=(X=-980,Y=400,Z=0),JointName_27_#6="pelvis")
    GravityOverrideParams_DBNO_Carried_41_#16:Gravity Override Setting Struct Structure=(WindFrequency_28_#2=(X=0,Y=0,Z=0),WindAmplitude_31_#3=(X=0,Y=0,Z=0),WindOffset_32_#4=(X=0,Y=0,Z=0),GravityOverride_15_#5=(X=-980,Y=500,Z=-500),JointName_27_#6="pelvis")
    GravityOverrideParams_OnGround_Falling_45_#17:Gravity Override Setting Struct Structure=(WindFrequency_28_#2=(X=0,Y=0,Z=0),WindAmplitude_31_#3=(X=0,Y=0,Z=0),WindOffset_32_#4=(X=0,Y=0,Z=0),GravityOverride_15_#5=(X=0,Y=0,Z=-980),JointName_27_#6="root")
    GravityOverrideParams_Ziplining_46_#18:Gravity Override Setting Struct Structure=(WindFrequency_28_#2=(X=0,Y=0,Z=0),WindAmplitude_31_#3=(X=0,Y=0,Z=0),WindOffset_32_#4=(X=0,Y=0,Z=0),GravityOverride_15_#5=(X=0,Y=0,Z=-980),JointName_27_#6="root")
    GravityOverrideParams_Skydive_Surfing_49_#19:Gravity Override Setting Struct Structure=(WindFrequency_28_#2=(X=1.5,Y=7,Z=8),WindAmplitude_31_#3=(X=200,Y=-500,Z=400),WindOffset_32_#4=(X=0,Y=0,Z=0),GravityOverride_15_#5=(X=500,Y=-300,Z=-980),JointName_27_#6="root")
    GravityOverrideParams_Riding_Standing_54_#20:Gravity Override Setting Struct Structure=(WindFrequency_28_#2=(X=0,Y=0,Z=0),WindAmplitude_31_#3=(X=0,Y=0,Z=0),WindOffset_32_#4=(X=0,Y=0,Z=0),GravityOverride_15_#5=(X=0,Y=-350,Z=-980),JointName_27_#6="root")
    GravityOverrideParams_Riding_Moving_55_#21:Gravity Override Setting Struct Structure=(WindFrequency_28_#2=(X=0,Y=0,Z=0),WindAmplitude_31_#3=(X=0,Y=0,Z=0),WindOffset_32_#4=(X=0,Y=0,Z=0),GravityOverride_15_#5=(X=0,Y=-450,Z=-450),JointName_27_#6="root")
    GravityOverrideParams_Riding_Moving_Targeting_58_#22:Gravity Override Setting Struct Structure=(WindFrequency_28_#2=(X=0,Y=0,Z=0),WindAmplitude_31_#3=(X=0,Y=0,Z=0),WindOffset_32_#4=(X=0,Y=0,Z=0),GravityOverride_15_#5=(X=0,Y=-200,Z=-980),JointName_27_#6="root")
    GravityOverrideParams_Tactical_Sprint_57_#23:Gravity Override Setting Struct Structure=(WindFrequency_28_#2=(X=0,Y=0,Z=0),WindAmplitude_31_#3=(X=0,Y=0,Z=0),WindOffset_32_#4=(X=0,Y=0,Z=0),GravityOverride_15_#5=(X=0,Y=-450,Z=-500),JointName_27_#6="root")
    GravityOverrideParams_Rail_Grinding_61_#24:Gravity Override Setting Struct Structure=(WindFrequency_28_#2=(X=0,Y=0,Z=0),WindAmplitude_31_#3=(X=0,Y=0,Z=0),WindOffset_32_#4=(X=0,Y=0,Z=0),GravityOverride_15_#5=(X=0,Y=0,Z=-980),JointName_27_#6="root")
    GravityOverrideParams_DBNO_Tumbling_67_#25:Gravity Override Setting Struct Structure=(WindFrequency_28_#2=(X=0,Y=0,Z=0),WindAmplitude_31_#3=(X=0,Y=0,Z=0),WindOffset_32_#4=(X=0,Y=0,Z=0),GravityOverride_15_#5=(X=-2000,Y=0,Z=0),JointName_27_#6="pelvis")
    GravityOverrideParams_DBNO_RollFwd_68_#26:Gravity Override Setting Struct Structure=(WindFrequency_28_#2=(X=0,Y=0,Z=0),WindAmplitude_31_#3=(X=0,Y=0,Z=0),WindOffset_32_#4=(X=0,Y=0,Z=0),GravityOverride_15_#5=(X=-2000,Y=0,Z=0),JointName_27_#6="pelvis")
    GravityOverrideParams_DBNO_RollBwd_69_#27:Gravity Override Setting Struct Structure=(WindFrequency_28_#2=(X=0,Y=0,Z=0),WindAmplitude_31_#3=(X=0,Y=0,Z=0),WindOffset_32_#4=(X=0,Y=0,Z=0),GravityOverride_15_#5=(X=-4000,Y=0,Z=0),JointName_27_#6="pelvis")
    GravityOverrideParams_DBNO_Impulse_70_#28:Gravity Override Setting Struct Structure=(WindFrequency_28_#2=(X=0,Y=0,Z=0),WindAmplitude_31_#3=(X=0,Y=0,Z=0),WindOffset_32_#4=(X=0,Y=0,Z=0),GravityOverride_15_#5=(X=-2000,Y=0,Z=0),JointName_27_#6="pelvis")
    RBAN Params
    DebugDraw
    DebugJointName
   ) -> (then, GravityOverride, CurrentStateOut, GravityOverrideSettingsOut, RBAN SettingsOut)
   > Set Gravity_Override(Gravity_Override=@n5)
   n6 = Branch(execute=@n1.then_2, Condition="float > float"(A=Get PawnSpeed2D(), B=50) {is=title K2Node_CallFunction})
   > Set move_alpha(move_alpha:double=1) -> (Output_Get:double)
   > Set strap_crouch_alpha(strap_crouch_alpha:double=0) -> (Output_Get:double)
   Set move_alpha(execute=@n6.else, move_alpha=0)
   > Set strap_crouch_alpha(strap_crouch_alpha=1)
   Branch(execute=@n1.then_4, Condition=Get bIsInFrontEnd())
   n7 = Set thigh_fwd_l(
    execute=@n1.then_3
    thigh_fwd_l:double=GetAlphaFromReferenceTransformDelta(
     SkeletalMeshComponent=Get Skel_Mesh()
     BoneName=thigh_l
     BaseName=pelvis
     DeltaComponentType=QuaternionTwist
     InputRange=(X=25,Y=80)
     TwistAxis="0,0,1"
    )
   ) -> (Output_Get:double)
   > Set inv_thigh_fwd_l(inv_thigh_fwd_l:double=float - float(A=1, B=@n7)) -> (Output_Get:double)
   > Set thigh_bck_l(thigh_bck_l:double=GetAlphaFromReferenceTransformDelta(
     SkeletalMeshComponent=Get Skel_Mesh()
     BoneName=thigh_l
     BaseName=pelvis
     DeltaComponentType=QuaternionTwist
     InputRange=(X=0,Y=-45)
     TwistAxis="0,0,1"
    )) -> (Output_Get:double)
   > n8 = Set thigh_fwd_r(thigh_fwd_r:double=GetAlphaFromReferenceTransformDelta(
     SkeletalMeshComponent=Get Skel_Mesh()
     BoneName=thigh_r
     BaseName=pelvis
     DeltaComponentType=QuaternionTwist
     InputRange=(X=25,Y=100)
     TwistAxis="0,0,1"
    )) -> (Output_Get:double)
   > Set inv_thigh_fwd_r(inv_thigh_fwd_r:double=float - float(A=1, B=@n8)) -> (Output_Get:double)
   > Set thigh_bck_r(thigh_bck_r:double=GetAlphaFromReferenceTransformDelta(
     SkeletalMeshComponent=Get Skel_Mesh()
     BoneName=thigh_r
     BaseName=pelvis
     DeltaComponentType=QuaternionTwist
     InputRange=(X=0,Y=-45)
     TwistAxis="0,0,1"
    )) -> (Output_Get:double)
   > n9 = Branch(Condition=Get bIsCrouching())
   > n10 = Branch(Condition="float > float"(A=Get PawnSpeed2D(), B=50))
   > Set thigh_fwd_l(thigh_fwd_l=0)
   > Set thigh_fwd_crouch_l(thigh_fwd_crouch_l:double=1) -> (Output_Get:double)
   Set thigh_fwd_crouch_l(execute=@n10.else, thigh_fwd_crouch_l=0)
   Set thigh_fwd_crouch_l(execute=@n9.else, thigh_fwd_crouch_l=0)
   Event BlueprintInitializeAnimation()
   > Set Skel_Mesh(Skel_Mesh=GetPartSkeletalMeshComponent())
   Event OnBeginCrouching()
   > Set crouch alpha(crouch alpha:double=1) -> (Output_Get:double)
   > Set crouch alpha inv(crouch alpha inv:double=0) -> (Output_Get:double)
   Event OnEndCrouching()
   > Set crouch alpha(crouch alpha=0)
   > Set crouch alpha inv(crouch alpha inv=1)
   Gravity Override State Function DEPRECATED!(
    execute
    CharacterPartAnimInstance=Self-Reference()
    GameTimeInSeconds=GetGameTimeInSeconds()
    GravityOverrideParams:"Gravity Override Params Struct Structure (by ref)"=Get Gravity Override Params()
    RBAN Params
    DebugDraw
    DebugJointName
   ) -> (then, GravityOverride, CurrentStateOut, GravityOverrideSettingsOut, RBAN SettingsOut)
   Get bCanPlayCustomAnimations() -> (bCanPlayCustomAnimations:bool)
  ]}
]
variables [
 crouch alpha {type=number}
 crouch alpha inv {type=number}
 fe_alpha {type=number}
 fe_alpha_inv {type=number}
 Gravity Override Params {type=GravityOverrideParamsStruct, default {skydive_DivingDown {windAmplitude=<0 0 0>}, skydive_Parachuting {gravityOverride=<0 0 -920>}}}
 Gravity Override Params_0 {type=GravityOverrideParamsStruct, default {front_End {windAmplitude=<200 -500 400>, windFrequency=<10000 20000 20000>}}}
 Gravity_Override {type=Vector}
 Gravity_Override_FE {type=Vector}
 inv_thigh_fwd_l {type=number}
 inv_thigh_fwd_r {type=number}
 move_alpha {type=number}
 Skel_Mesh {type=SkeletalMeshComponent}
 strap_crouch_alpha {type=number}
 thigh_bck_l {type=number}
 thigh_bck_r {type=number}
 thigh_fwd_crouch_l {type=number}
 thigh_fwd_l {type=number}
 thigh_fwd_r {type=number}
]
