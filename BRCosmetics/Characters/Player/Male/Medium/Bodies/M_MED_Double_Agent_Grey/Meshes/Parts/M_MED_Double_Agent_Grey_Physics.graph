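# Physics asset for the Double Agent Grey male medium body part, simulating the dangling hoodie drawstrings.
# Capsule bodies form a three-joint left chain and a four-joint right chain, both anchored to a large spine_04 torso capsule.
# Constraints leave the first swing axis free and tightly limit twist.
# The second swing is free at both roots and the right tip, and narrowly limited elsewhere, more tightly on the right.
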PhysicsAsset /BRCosmetics/Characters/Player/Male/Medium/Bodies/M_MED_Double_Agent_Grey/Meshes/Parts/M_MED_Double_Agent_Grey_Physics
bodies {
 dyn_L_hoodieString_jnt_1 [{shapeType=Capsule, length=3, radius=0.5, center=<1.1509 0 0>, rotation=<rot 90 0 0>}]
 dyn_L_hoodieString_jnt_2 [{shapeType=Capsule, length=3, radius=2, center=<1.1509 0 0>, rotation=<rot 90 0 0>}]
 dyn_L_hoodieString_jnt_3 [{shapeType=Capsule, length=3, radius=2.1, center=<1.1509 0 0>, rotation=<rot 90 0 0>}]
 dyn_R_hoodieString_jnt_1 [{shapeType=Capsule, length=3, radius=0.5, center=<1.1509 0 0>, rotation=<rot 90 0 0>}]
 dyn_R_hoodieString_jnt_2 [{shapeType=Capsule, length=3, radius=2, center=<1.1509 0 0>, rotation=<rot 90 0 0>}]
 dyn_R_hoodieString_jnt_3 [{shapeType=Capsule, length=3, radius=2.1, center=<1.1509 0 0>, rotation=<rot 90 0 0>}]
 dyn_R_hoodieString_jnt_4 [{shapeType=Capsule, length=3, radius=2.1, center=<1.1509 0.4429 0>, rotation=<rot 90 0 0>}]
 spine_04 [{shapeType=Capsule, length=12, radius=12, center=<-0.5802 1.0103 0>, rotation=<rot 0 85.4073 -89.9999>}]
}
constraints [
 {
  bone1Name=dyn_L_hoodieString_jnt_2
  bone2Name=dyn_L_hoodieString_jnt_1
  swing1LimitDegrees=90
  swing1Motion=Free
  swing2LimitDegrees=20
  twistLimitDegrees=5
 }
 {
  bone1Name=dyn_L_hoodieString_jnt_3
  bone2Name=dyn_L_hoodieString_jnt_2
  swing1LimitDegrees=90
  swing1Motion=Free
  swing2LimitDegrees=20
  twistLimitDegrees=5
 }
 {
  bone1Name=dyn_R_hoodieString_jnt_2
  bone2Name=dyn_R_hoodieString_jnt_1
  swing1LimitDegrees=90
  swing1Motion=Free
  swing2LimitDegrees=10
  twistLimitDegrees=5
 }
 {
  bone1Name=dyn_R_hoodieString_jnt_3
  bone2Name=dyn_R_hoodieString_jnt_2
  swing1LimitDegrees=90
  swing1Motion=Free
  swing2LimitDegrees=10
  twistLimitDegrees=5
 }
 {
  bone1Name=dyn_R_hoodieString_jnt_4
  bone2Name=dyn_R_hoodieString_jnt_3
  swing1LimitDegrees=90
  swing1Motion=Free
  swing2LimitDegrees=90
  swing2Motion=Free
  twistLimitDegrees=5
 }
 {
  bone1Name=dyn_L_hoodieString_jnt_1
  bone2Name=spine_04
  swing1LimitDegrees=90
  swing1Motion=Free
  swing2LimitDegrees=90
  swing2Motion=Free
  twistLimitDegrees=5
 }
 {
  bone1Name=dyn_R_hoodieString_jnt_1
  bone2Name=spine_04
  swing1LimitDegrees=90
  swing1Motion=Free
  swing2LimitDegrees=90
  swing2Motion=Free
  twistLimitDegrees=5
 }
]
