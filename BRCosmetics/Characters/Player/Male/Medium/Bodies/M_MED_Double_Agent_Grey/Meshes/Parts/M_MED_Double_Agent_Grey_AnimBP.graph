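# Post-process-style AnimBP for the Double Agent Grey male medium body part, adding secondary motion to its hood and hoodie.
# Copies the attached parent's pose, offsets hood and string bones, runs spring-driven AnimDynamics on both hood sides, then RigidBody simulation relative to spine_05.
# The center-hood AnimDynamics node is unconnected.
# The event graph derives left/right head-twist alphas against spine_05, which the AnimGraph does not read.
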AnimBlueprint /BRCosmetics/Characters/Player/Male/Medium/Bodies/M_MED_Double_Agent_Grey/Meshes/Parts/M_MED_Double_Agent_Grey_AnimBP
target_skeleton=&/M_MED_Double_Agent_Grey_Skeleton
graphs [
 AnimGraph {nodes [
   Anim Dynamics - Bone: dyn_center_hood_jnt(Alpha=1, ExternalForce="0,0,0") {
    bGravityOverrideInSimSpace=true
    bOverrideLinearDamping=true
    linearDampingOverride=1
    simulationSpace=RootRelative
    alphaBoolBlend {blendOption=Linear}
    boundBone {boneName=dyn_center_hood_jnt}
    componentAppliedLinearAccClamp=<5000 5000 5000>
    componentLinearAccScale=<2 2 2>
    physicsBodyDefinitions [{
      sphereCollisionRadius=0
      boundBone {boneName=dyn_center_hood_jnt}
      boxExtents=<1 1 1>
      constraintSetup {angularLimitsMax=<0 0 0>, angularLimitsMin=<0 0 0>, angularTarget=<0 0 0>, linearAxesMax=<0 1 0>, linearAxesMin=<0 -1 0>}
      localJointOffset=<0 0 0>
     }]
    relativeSpaceBone {boneName=neck_01}
   }
   n1 = "Transform (Modify) Bone - Bone: dyn_L_hood_jnt"(
    ComponentPose=Local To Component(LocalPose=Copy Pose From Mesh() {bUseAttachedParent=true})
    Alpha=1
    Translation="3,-2,0"
    Rotation="0,0,0"
    Scale="1,1,1"
   ) {
    translationMode=BMM_Additive
    translationSpace=BCS_BoneSpace
    alphaBoolBlend {blendOption=Linear}
    boneToModify {boneName=dyn_L_hood_jnt}
    translation=<3 -2>
   }
   n2 = "Transform (Modify) Bone - Bone: dyn_R_hood_jnt"(ComponentPose=@n1, Alpha=1, Translation="3,6,0", Rotation="0,0,0", Scale="1,1,1") {
    translationMode=BMM_Additive
    translationSpace=BCS_BoneSpace
    alphaBoolBlend {blendOption=Linear}
    boneToModify {boneName=dyn_R_hood_jnt}
    translation=<3 6>
   }
   n3 = "Transform (Modify) Bone - Bone: dyn_R_hood_jnt"(ComponentPose=@n2, Alpha=1, Translation="-2,0,0", Rotation="0,0,0", Scale="1,1,1") {
    translationMode=BMM_Additive
    translationSpace=BCS_BoneSpace
    alphaBoolBlend {blendOption=Linear}
    boneToModify {boneName=dyn_R_hood_jnt}
    translation {x=-2}
   }
   n4 = "Transform (Modify) Bone - Bone: dyn_L_hood_jnt"(ComponentPose=@n3, Alpha=1, Translation="-2,5,0", Rotation="0,0,0", Scale="1,1,1") {
    translationMode=BMM_Additive
    translationSpace=BCS_BoneSpace
    alphaBoolBlend {blendOption=Linear}
    boneToModify {boneName=dyn_L_hood_jnt}
    translation=<-2 5>
   }
   n5 = Anim Dynamics - Bone: dyn_L_hood_jnt(ComponentPose=@n4, Alpha=0.6, ExternalForce="0,0,0") {
    bGravityOverrideInSimSpace=true
    bLinearSpring=true
    bOverrideLinearDamping=true
    bUseGravityOverride=true
    linearDampingOverride=1
    linearSpringConstant=60
    simulationSpace=RootRelative
    alphaBoolBlend {blendOption=Linear}
    boundBone {boneName=dyn_L_hood_jnt}
    componentAppliedLinearAccClamp=<5000 5000 5000>
    componentLinearAccScale=<1 1 1>
    componentLinearVelScale=<2 2 2>
    gravityOverride {x=100, z=-980}
    physicsBodyDefinitions [{
      sphereCollisionRadius=0
      boundBone {boneName=dyn_L_hood_jnt}
      boxExtents=<1 1 1>
      constraintSetup {angularLimitsMax=<0 0 0>, angularLimitsMin=<0 0 0>, angularTarget=<0 0 0>, linearAxesMax=<1 0 0>}
      localJointOffset=<0 0 0>
     }]
    relativeSpaceBone {boneName=neck_01}
   }
   n6 = Anim Dynamics - Bone: dyn_R_hood_jnt(ComponentPose=@n5, Alpha=0.6, ExternalForce="0,0,0") {
    bGravityOverrideInSimSpace=true
    bLinearSpring=true
    bOverrideLinearDamping=true
    bUseGravityOverride=true
    linearDampingOverride=1
    linearSpringConstant=60
    simulationSpace=RootRelative
    alphaBoolBlend {blendOption=Linear}
    boundBone {boneName=dyn_R_hood_jnt}
    componentAppliedLinearAccClamp=<5000 5000 5000>
    componentLinearAccScale=<1 1 1>
    componentLinearVelScale=<2 2 2>
    gravityOverride {x=-100, z=-980}
    physicsBodyDefinitions [{
      sphereCollisionRadius=0
      boundBone {boneName=dyn_R_hood_jnt}
      boxExtents=<1 1 1>
      constraintSetup {angularLimitsMax=<0 0 0>, angularLimitsMin=<0 0 0>, angularTarget=<0 0 0>, linearAxesMax=<1 0 0>}
      localJointOffset=<0 0 0>
     }]
    relativeSpaceBone {boneName=neck_01}
   }
   n7 = "Transform (Modify) Bone - Bone: alt_R_hoodieString_jnt"(ComponentPose=@n6, Alpha=1, Translation="0,0,0", Rotation="0,0,-8", Scale="1,1,1") {
    alpha=0
    rotationMode=BMM_Additive
    translationMode=BMM_Additive
    alphaBoolBlend {blendOption=Linear}
    boneToModify {boneName=alt_R_hoodieString_jnt}
    rotation {roll=-8}
   }
   n8 = "Transform (Modify) Bone - Bone: alt_L_hoodieString_jnt"(ComponentPose=@n7, Alpha=1, Translation="0,0,0", Rotation="0,0,-8", Scale="1,1,1") {
    alpha=0
    rotationMode=BMM_Additive
    translationMode=BMM_Additive
    alphaBoolBlend {blendOption=Linear}
    boneToModify {boneName=alt_L_hoodieString_jnt}
    rotation {roll=-8}
   }
   n9 = RigidBody(ComponentPose=@n8) {
    bClampLinearTranslationLimitToRefPose=true
    bOverrideWorldGravity=true
    overridePhysicsAsset=&/M_MED_Double_Agent_Grey_Physics
    simulationSpace=BaseBoneSpace
    alphaBoolBlend {blendOption=Linear}
    baseBoneRef {boneName=spine_05}
    componentAppliedLinearAccClamp=<4000 4000 4000>
    componentLinearAccScale=<0.4 0.3 0.3>
    componentLinearVelScale=<0.4 0.3 0.3>
    overrideWorldGravity {z=-980}
   }
   Output Pose(Result=Component To Local(ComponentPose=@n9)) {layerGroup=None}
   "To Float (Boolean)"(InBool=Get bIsInFrontEnd())
  ]}
 EventGraph {nodes [
   Event BlueprintUpdateAnimation()
   > Get!(execute:Exec) -> (then:Exec, else:Exec, Skel_Mesh:Skeletal Mesh Component Object Reference)
   > Sequence!(execute) -> (then_0, then_1, then_2:Exec, then_3:Exec)
   > Set head_left(head_left:double=GetAlphaFromReferenceTransformDelta(
     SkeletalMeshComponent=Get Skel_Mesh()
     BoneName=head
     BaseName=spine_05
     DeltaComponentType=QuaternionTwist
     InputRange=(X=0,Y=55)
     TwistAxis="1,0,0"
    )) -> (Output_Get:double)
   > Set head_right(head_right:double=GetAlphaFromReferenceTransformDelta(
     SkeletalMeshComponent=Get Skel_Mesh()
     BoneName=head
     BaseName=spine_05
     DeltaComponentType=QuaternionTwist
     InputRange=(X=0,Y=-55)
     TwistAxis="1,0,0"
    )) -> (Output_Get:double)
   Event BlueprintInitializeAnimation()
   > Set Skel_Mesh(Skel_Mesh=GetPartSkeletalMeshComponent())
  ]}
]
variables [
 head_left {type=number}
 head_right {type=number}
 Skel_Mesh {type=SkeletalMeshComponent}
]
